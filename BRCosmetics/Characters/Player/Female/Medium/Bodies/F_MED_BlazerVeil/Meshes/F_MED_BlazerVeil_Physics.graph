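# Physics asset for the F_MED_BlazerVeil female medium player body, set up for the blazer's dynamic sleeves.
# Each sleeve has top, bottom and back three-capsule chains, each rooted on a small MSTR body that links to the lower arm with free angular motion.
# Chain joints have angular limits, tightest at some roots and on back-chain twist.
# Capsules on hands, forearms, pelvis and thighs, joined by free constraints, supply the surrounding body shapes.
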PhysicsAsset /BRCosmetics/Characters/Player/Female/Medium/Bodies/F_MED_BlazerVeil/Meshes/F_MED_BlazerVeil_Physics
bodies {
 dyn_sleeve_l_bk_01 [{shapeType=Capsule, length=6, radius=3.5, shapeName=dyn_sleeve_l_bk_01_capsule, center=<5 -0.5 0>, rotation=<rot 90 0 0>}]
 dyn_sleeve_l_bk_02 [{shapeType=Capsule, length=6, radius=3.5, shapeName=dyn_sleeve_l_bk_02_capsule, center=<5 -0.5 0>, rotation=<rot 90 0 0>}]
 dyn_sleeve_l_bk_03 [{shapeType=Capsule, length=6, radius=3.5, shapeName=dyn_sleeve_l_bk_03_capsule, center=<5 -0.5 0>, rotation=<rot 90 0 0>}]
 dyn_sleeve_l_bk_MSTR [{shapeType=Capsule, length=0, radius=1, shapeName=dyn_sleeve_l_bk_MSTR_capsule}]
 dyn_sleeve_l_bot_01 [{shapeType=Capsule, length=6, radius=3.5, shapeName=dyn_sleeve_l_bot_01_capsule, center=<3 0 0>, rotation=<rot 90 0 0>}]
 dyn_sleeve_l_bot_02 [{shapeType=Capsule, length=6, radius=3.5, shapeName=dyn_sleeve_l_bot_02_capsule, center=<3 0 0>, rotation=<rot 90 0 0>}]
 dyn_sleeve_l_bot_03 [{shapeType=Capsule, length=6, radius=3.5, shapeName=dyn_sleeve_l_bot_03_capsule, center=<3 0 0>, rotation=<rot 90 0 0>}]
 dyn_sleeve_l_bot_MSTR [{shapeType=Capsule, length=0, radius=1, shapeName=dyn_sleeve_l_bot_MSTR_capsule}]
 dyn_sleeve_l_top_01 [{shapeType=Capsule, length=6, radius=3.5, shapeName=dyn_sleeve_l_bk_01_capsule, center=<3 2 1.5>, rotation=<rot 90 0 0>}]
 dyn_sleeve_l_top_02 [{shapeType=Capsule, length=11, radius=3.5, shapeName=dyn_sleeve_l_top_02_capsule, center=<3 0 2>, rotation=<rot 90 0 0>}]
 dyn_sleeve_l_top_03 [{shapeType=Capsule, length=11, radius=3.5, shapeName=dyn_sleeve_l_top_03_capsule, center=<2 0 2>, rotation=<rot 90 0 0>}]
 dyn_sleeve_l_top_MSTR [{shapeType=Capsule, length=0, radius=1, shapeName=dyn_sleeve_l_top_MSTR_capsule}]
 dyn_sleeve_r_bk_01 [{shapeType=Capsule, length=6, radius=3.5, shapeName=dyn_sleeve_r_bk_01_capsule, center=<-3 1.5 0>, rotation=<rot -90 0 0>}]
 dyn_sleeve_r_bk_02 [{shapeType=Capsule, length=6, radius=3.5, shapeName=dyn_sleeve_r_bk_02_capsule, center=<-3 1.5 0>, rotation=<rot -90 0 0>}]
 dyn_sleeve_r_bk_03 [{shapeType=Capsule, length=6, radius=3.5, shapeName=dyn_sleeve_r_bk_03_capsule, center=<-3 1.5 0>, rotation=<rot -90 0 0>}]
 dyn_sleeve_r_bk_MSTR [{shapeType=Capsule, length=0, radius=1, shapeName=dyn_sleeve_r_bk_MSTR_capsule}]
 dyn_sleeve_r_bot_01 [{shapeType=Capsule, length=6, radius=3.5, shapeName=dyn_sleeve_r_bot_01_capsule, center=<-3 0 0>, rotation=<rot -90 0 0>}]
 dyn_sleeve_r_bot_02 [{shapeType=Capsule, length=6, radius=3.5, shapeName=dyn_sleeve_r_bot_02_capsule, center=<-3 0 0>, rotation=<rot -90 0 0>}]
 dyn_sleeve_r_bot_03 [{shapeType=Capsule, length=6, radius=3.5, shapeName=dyn_sleeve_r_bot_03_capsule, center=<-3 0 0>, rotation=<rot -90 0 0>}]
 dyn_sleeve_r_bot_MSTR [{shapeType=Capsule, length=0, radius=1, shapeName=dyn_sleeve_r_bot_MSTR_capsule}]
 dyn_sleeve_r_top_01 [{shapeType=Capsule, length=6, radius=3.5, shapeName=dyn_sleeve_r_bot_01_capsule, center=<-3 -2 -1.5>, rotation=<rot -90 180 180>}]
 dyn_sleeve_r_top_02 [{shapeType=Capsule, length=11, radius=3.5, shapeName=dyn_sleeve_r_bot_02_capsule, center=<-3 0 -1.5>, rotation=<rot -90 180 180>}]
 dyn_sleeve_r_top_03 [{shapeType=Capsule, length=11, radius=3.5, shapeName=dyn_sleeve_r_bot_02_capsule, center=<-3 0 -1.5>, rotation=<rot -90 180 180>}]
 dyn_sleeve_r_top_MSTR [{shapeType=Capsule, length=0, radius=1, shapeName=dyn_sleeve_r_top_MSTR_capsule}]
 hand_l [{
   shapeType=Capsule
   length=5.9315
   radius=3.5693
   shapeName=hand_l_capsule
   center=<3.3749 -0.9476 0.0934>
   rotation=<rot 11.7161 -90 -85.9729>
  }]
 hand_r [{
   shapeType=Capsule
   length=5.9716
   radius=3.6627
   shapeName=hand_r_capsule
   center=<-3.4341 1.0233 0.0998>
   rotation=<rot 8.5334 -90 -91.2856>
  }]
 lowerarm_l [{
   shapeType=Capsule
   length=22.7728
   radius=5.2272
   shapeName=lowerarm_l_capsule
   center=<10.3898 0.5176 0>
   rotation=<rot 90 -55.1755 -50.1756>
  }]
 lowerarm_r [{shapeType=Capsule, length=25.2187, radius=5.0671, shapeName=lowerarm_r_capsule, center=<-9.131 -0.2763 0>, rotation=<rot 90 0 9.9998>}]
 lowerarm_twist_02_l [{
   shapeType=Capsule
   length=6.5148
   radius=4.6062
   shapeName=lowerarm_twist_02_l_capsule
   center=<-4.7619 1.1194 0.3615>
   rotation=<rot -9.839 -96.5687 -93.6191>
  }]
 lowerarm_twist_02_r [{shapeType=Capsule, length=9.0453, radius=4.1809, shapeName=lowerarm_twist_02_r_capsule, rotation=<rot 90 -117.4458 -117.4463>}]
 pelvis [{shapeType=Capsule, length=16.7656, radius=13.0684, shapeName=pelvis_capsule, center=<-4.7356 0 0>, rotation=<rot 90 -27.9382 -27.9379>}]
 thigh_l [{shapeType=Capsule, length=39.4087, radius=7.5893, shapeName=thigh_l_capsule, center=<-18.1986 0 0>, rotation=<rot 90 0 0>}]
 thigh_r [{shapeType=Capsule, length=43.4133, radius=6.8579, shapeName=thigh_r_capsule, center=<18.4013 0 0.2461>, rotation=<rot 85 179.9999 179.9999>}]
}
constraints [
 {
  bone1Name=hand_l
  bone2Name=lowerarm_l
  swing1LimitDegrees=45
  swing1Motion=Free
  swing2LimitDegrees=45
  swing2Motion=Free
  twistLimitDegrees=45
  twistMotion=Free
 }
 {
  bone1Name=lowerarm_l
  bone2Name=pelvis
  swing1LimitDegrees=45
  swing1Motion=Free
  swing2LimitDegrees=45
  swing2Motion=Free
  twistLimitDegrees=45
  twistMotion=Free
 }
 {bone1Name=dyn_sleeve_l_top_01, bone2Name=dyn_sleeve_l_top_MSTR, swing1LimitDegrees=20, swing2LimitDegrees=20, twistLimitDegrees=2}
 {bone1Name=dyn_sleeve_l_top_03, bone2Name=dyn_sleeve_l_top_02, swing1LimitDegrees=45, swing2LimitDegrees=45, twistLimitDegrees=45}
 {bone1Name=dyn_sleeve_l_top_02, bone2Name=dyn_sleeve_l_top_01, swing1LimitDegrees=45, swing2LimitDegrees=45, twistLimitDegrees=45}
 {
  bone1Name=dyn_sleeve_l_top_MSTR
  bone2Name=lowerarm_l
  swing1LimitDegrees=45
  swing1Motion=Free
  swing2LimitDegrees=45
  swing2Motion=Free
  twistLimitDegrees=45
  twistMotion=Free
 }
 {bone1Name=dyn_sleeve_l_bot_03, bone2Name=dyn_sleeve_l_bot_02, swing1LimitDegrees=45, swing2LimitDegrees=45, twistLimitDegrees=45}
 {bone1Name=dyn_sleeve_l_bot_02, bone2Name=dyn_sleeve_l_bot_01, swing1LimitDegrees=45, swing2LimitDegrees=45, twistLimitDegrees=45}
 {
  bone1Name=dyn_sleeve_l_bot_MSTR
  bone2Name=lowerarm_l
  swing1LimitDegrees=45
  swing1Motion=Free
  swing2LimitDegrees=45
  swing2Motion=Free
  twistLimitDegrees=45
  twistMotion=Free
 }
 {bone1Name=dyn_sleeve_l_bot_01, bone2Name=dyn_sleeve_l_bot_MSTR, swing1LimitDegrees=45, swing2LimitDegrees=45, twistLimitDegrees=45}
 {bone1Name=dyn_sleeve_l_bk_03, bone2Name=dyn_sleeve_l_bk_02, swing1LimitDegrees=45, swing2LimitDegrees=45, twistLimitDegrees=2}
 {bone1Name=dyn_sleeve_l_bk_02, bone2Name=dyn_sleeve_l_bk_01, swing1LimitDegrees=45, swing2LimitDegrees=45, twistLimitDegrees=2}
 {bone1Name=dyn_sleeve_l_bk_01, bone2Name=dyn_sleeve_l_bk_MSTR, swing1LimitDegrees=5, swing2LimitDegrees=5, twistLimitDegrees=2}
 {
  bone1Name=dyn_sleeve_l_bk_MSTR
  bone2Name=lowerarm_l
  swing1LimitDegrees=45
  swing1Motion=Free
  swing2LimitDegrees=45
  swing2Motion=Free
  twistLimitDegrees=45
  twistMotion=Free
 }
 {
  bone1Name=lowerarm_r
  bone2Name=pelvis
  swing1LimitDegrees=45
  swing1Motion=Free
  swing2LimitDegrees=45
  swing2Motion=Free
  twistLimitDegrees=45
  twistMotion=Free
 }
 {
  bone1Name=dyn_sleeve_r_bk_MSTR
  bone2Name=lowerarm_r
  swing1LimitDegrees=45
  swing1Motion=Free
  swing2LimitDegrees=45
  swing2Motion=Free
  twistLimitDegrees=45
  twistMotion=Free
 }
 {bone1Name=dyn_sleeve_r_bk_01, bone2Name=dyn_sleeve_r_bk_MSTR, swing1LimitDegrees=10, swing2LimitDegrees=10, twistLimitDegrees=1}
 {bone1Name=dyn_sleeve_r_bk_02, bone2Name=dyn_sleeve_r_bk_01, swing1LimitDegrees=45, swing2LimitDegrees=45, twistLimitDegrees=2}
 {bone1Name=dyn_sleeve_r_bk_03, bone2Name=dyn_sleeve_r_bk_02, swing1LimitDegrees=45, swing2LimitDegrees=45, twistLimitDegrees=2}
 {
  bone1Name=dyn_sleeve_r_bot_MSTR
  bone2Name=lowerarm_r
  swing1LimitDegrees=45
  swing1Motion=Free
  swing2LimitDegrees=45
  swing2Motion=Free
  twistLimitDegrees=45
  twistMotion=Free
 }
 {bone1Name=dyn_sleeve_r_bot_01, bone2Name=dyn_sleeve_r_bot_MSTR, swing1LimitDegrees=45, swing2LimitDegrees=45, twistLimitDegrees=2}
 {bone1Name=dyn_sleeve_r_bot_02, bone2Name=dyn_sleeve_r_bot_01, swing1LimitDegrees=45, swing2LimitDegrees=45, twistLimitDegrees=45}
 {bone1Name=dyn_sleeve_r_bot_03, bone2Name=dyn_sleeve_r_bot_02, swing1LimitDegrees=45, swing2LimitDegrees=45, twistLimitDegrees=45}
 {
  bone1Name=hand_r
  bone2Name=lowerarm_r
  swing1LimitDegrees=45
  swing1Motion=Free
  swing2LimitDegrees=45
  swing2Motion=Free
  twistLimitDegrees=45
  twistMotion=Free
 }
 {
  bone1Name=thigh_l
  bone2Name=pelvis
  swing1LimitDegrees=45
  swing1Motion=Free
  swing2LimitDegrees=45
  swing2Motion=Free
  twistLimitDegrees=45
  twistMotion=Free
 }
 {
  bone1Name=thigh_r
  bone2Name=pelvis
  swing1LimitDegrees=45
  swing1Motion=Free
  swing2LimitDegrees=45
  swing2Motion=Free
  twistLimitDegrees=45
  twistMotion=Free
 }
 {
  bone1Name=lowerarm_twist_02_l
  bone2Name=lowerarm_l
  swing1LimitDegrees=45
  swing1Motion=Free
  swing2LimitDegrees=45
  swing2Motion=Free
  twistLimitDegrees=45
  twistMotion=Free
 }
 {
  bone1Name=lowerarm_twist_02_r
  bone2Name=lowerarm_r
  swing1LimitDegrees=45
  swing1Motion=Free
  swing2LimitDegrees=45
  swing2Motion=Free
  twistLimitDegrees=45
  twistMotion=Free
 }
 {
  bone1Name=dyn_sleeve_r_top_MSTR
  bone2Name=lowerarm_r
  swing1LimitDegrees=45
  swing1Motion=Free
  swing2LimitDegrees=45
  swing2Motion=Free
  twistLimitDegrees=45
  twistMotion=Free
 }
 {bone1Name=dyn_sleeve_r_top_01, bone2Name=dyn_sleeve_r_top_MSTR, swing1LimitDegrees=30, swing2LimitDegrees=10, twistLimitDegrees=10}
 {bone1Name=dyn_sleeve_r_top_02, bone2Name=dyn_sleeve_r_top_01, swing1LimitDegrees=45, swing2LimitDegrees=25, twistLimitDegrees=45}
 {bone1Name=dyn_sleeve_r_top_03, bone2Name=dyn_sleeve_r_top_02, swing1LimitDegrees=45, swing2LimitDegrees=25, twistLimitDegrees=45}
]
