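# Animation Blueprint for the F_MED_BlazerVeil female body cosmetic, adding secondary motion to dynamic clothing bones such as sleeves, jacket and skirt.
# From cached StartupCache pose, bRigidBodyEnabled blends between forearm-constrained pelvis-space RigidBody simulation and a per-bone spring AnimDynamics sleeve chain.
# Weighted constraints then correct the blended result, tying sleeve bones to each other and the arms, and jacket bones to spine, clavicles and arms.
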AnimBlueprint /BRCosmetics/Characters/Player/Female/Medium/Bodies/F_MED_BlazerVeil/Meshes/F_MED_BlazerVeil_AnimBP
target_skeleton=&/F_MED_BlazerVeil_Skeleton
graphs [
 AnimGraph {nodes [
   n1 = "Constraint - dyn_skirt_bk_01 (PP)"!(ComponentPose, Alpha=0.4, ConstraintWeights_0=0.5, ConstraintWeights_1:float=0.5) -> (Pose) {
    alphaBoolBlend {blendOption=Linear}
    boneToModify {boneName=dyn_skirt_bk_01}
    constraintSetup [{targetBone {boneName=thigh_l}}, {targetBone {boneName=thigh_r}}]
    constraintWeights [1,1]
   }
   n2 = "Constraint - dyn_skirt_bk_02 (PP)"!(ComponentPose=@n1, Alpha=0.4, ConstraintWeights_0=0.5, ConstraintWeights_1:float=0.5) -> (Pose) {
    alphaBoolBlend {blendOption=Linear}
    boneToModify {boneName=dyn_skirt_bk_02}
    constraintSetup [{targetBone {boneName=thigh_l}}, {targetBone {boneName=thigh_r}}]
    constraintWeights [1,1]
   }
   "Transform (Modify) Bone - Bone: dyn_skirt_bk_01"!(ComponentPose=@n2, Alpha=1, Translation="0,0,0", Rotation="0.000,-5.000,0.000") -> (Pose) {
    rotationMode=BMM_Additive
    rotationSpace=BCS_BoneSpace
    translationMode=BMM_Additive
    translationSpace=BCS_BoneSpace
    alphaBoolBlend {blendOption=Linear}
    boneToModify {boneName=dyn_skirt_bk_01}
    rotation=<rot 0 -4.9999 0>
   }
   RigidBody!(ComponentPose, Alpha, OverrideWorldGravity:Vector="0,0,0", ExternalForce) -> (Pose) {
    bClampLinearTranslationLimitToRefPose=true
    bOverrideWorldGravity=true
    overridePhysicsAsset=&/F_MED_BlazerVeil_Physics
    simulationSpace=BaseBoneSpace
    alphaBoolBlend {blendOption=Linear}
    baseBoneRef {boneName=spine_05}
    componentAppliedLinearAccClamp=<3000 3000 3000>
    simSpaceSettings {
     maxAngularAcceleration=500
     maxAngularVelocity=3
     maxLinearAcceleration=400
     maxLinearVelocity=250
     velocityScaleZ=0
     worldAlpha=1
     externalLinearDragV=<0.1 0.1>
    }
   }
   RigidBody(OverrideWorldGravity="0,0,-980") {
    bClampLinearTranslationLimitToRefPose=true
    bOverrideWorldGravity=true
    overridePhysicsAsset=&F_MED_BlazerVeil_Physics
    simulationSpace=BaseBoneSpace
    alphaBoolBlend {blendOption=Linear}
    baseBoneRef {boneName=pelvis}
    componentAppliedLinearAccClamp=<3000 3000 3000>
    componentLinearAccScale=<0.5 0.3 0.3>
    componentLinearVelScale=<0.5 0.3 0.3>
    overrideWorldGravity {z=-980}
    simSpaceSettings {
     maxAngularAcceleration=200
     maxAngularVelocity=5
     maxLinearAcceleration=1000
     maxLinearVelocity=200
     worldAlpha=0.55
     externalLinearDragV=<1 1>
    }
   }
   RigidBody(OverrideWorldGravity="0,0,-980") {
    bOverrideWorldGravity=true
    overridePhysicsAsset=&F_MED_BlazerVeil_Physics
    simulationSpace=BaseBoneSpace
    alphaBoolBlend {blendOption=Linear}
    baseBoneRef {boneName=pelvis}
    componentAppliedLinearAccClamp=<3500 3500 3500>
    componentLinearAccScale=<0.7 0.5 0.7>
    componentLinearVelScale=<0.7 0.5 0.7>
    simSpaceSettings {maxAngularAcceleration=3000, maxAngularVelocity=0, maxLinearAcceleration=1000, maxLinearVelocity=2000, velocityScaleZ=0.2, worldAlpha=1}
   }
   n3 = Anim Dynamics - Bone: dyn_sleeve_l_top_01!(
    ComponentPose=Local To Component(LocalPose=Use cached pose 'StartupCache'())
    Alpha=1
    GravityOverride:Vector="0,0,-980"
    ExternalForce="0,0,0"
   ) -> (Pose) {
    angularDampingOverride=0.9999
    angularSpringConstant=500
    bAngularSpring=true
    bLinearSpring=true
    bOverrideAngularDamping=true
    bUseGravityOverride=true
    bUseSphericalLimits=true
    linearSpringConstant=50
    simulationSpace=BoneRelative
    alphaBoolBlend {blendOption=Linear}
    boundBone {boneName=dyn_sleeve_l_top_01}
    componentAppliedLinearAccClamp=<1500 1500 0>
    componentLinearAccScale=<0.3 0.3 0.3>
    componentLinearVelScale=<0.3 0.3 0.3>
    gravityOverride {z=-980}
    physicsBodyDefinitions [{
      sphereCollisionRadius=0
      boundBone {boneName=dyn_sleeve_l_top_01}
      boxExtents=<5 20 20>
      constraintSetup {angularLimitsMax=<2 1 45>, angularLimitsMin=<-2 -30 -45>, angularTarget=<0 0 0>}
      localJointOffset=<3 0 0>
     }]
    relativeSpaceBone {boneName=root}
   }
   n4 = Anim Dynamics - Bone: dyn_sleeve_l_top_02!(ComponentPose=@n3, Alpha=1, GravityOverride:Vector="0,0,-980", ExternalForce="0,0,0") -> (Pose) {
    angularDampingOverride=0.9999
    angularSpringConstant=500
    bAngularSpring=true
    bLinearSpring=true
    bOverrideAngularDamping=true
    bUseGravityOverride=true
    bUseSphericalLimits=true
    linearSpringConstant=50
    simulationSpace=BoneRelative
    alphaBoolBlend {blendOption=Linear}
    boundBone {boneName=dyn_sleeve_l_top_02}
    componentAppliedLinearAccClamp=<1500 1500 0>
    componentLinearAccScale=<0.3 0.3 0.3>
    componentLinearVelScale=<0.3 0.3 0.3>
    gravityOverride {z=-980}
    physicsBodyDefinitions [{
      sphereCollisionRadius=0
      boundBone {boneName=dyn_sleeve_l_top_02}
      boxExtents=<5 20 20>
      constraintSetup {angularLimitsMax=<2 1 45>, angularLimitsMin=<-2 -30 -45>, angularTarget=<0 0 0>}
      localJointOffset=<3 0 0>
     }]
    relativeSpaceBone {boneName=root}
   }
   n5 = Anim Dynamics - Bone: dyn_sleeve_l_top_03!(ComponentPose=@n4, Alpha=1, GravityOverride:Vector="0,0,-980", ExternalForce="0,0,0") -> (Pose) {
    angularDampingOverride=0.9999
    angularSpringConstant=500
    bAngularSpring=true
    bLinearSpring=true
    bOverrideAngularDamping=true
    bUseGravityOverride=true
    bUseSphericalLimits=true
    linearSpringConstant=50
    simulationSpace=BoneRelative
    alphaBoolBlend {blendOption=Linear}
    boundBone {boneName=dyn_sleeve_l_top_03}
    componentAppliedLinearAccClamp=<1500 1500 0>
    componentLinearAccScale=<0.3 0.3 0.3>
    componentLinearVelScale=<0.3 0.3 0.3>
    gravityOverride {z=-980}
    physicsBodyDefinitions [{
      sphereCollisionRadius=0
      boundBone {boneName=dyn_sleeve_l_top_03}
      boxExtents=<5 20 20>
      constraintSetup {angularLimitsMax=<2 1 45>, angularLimitsMin=<-2 -30 -45>, angularTarget=<0 0 0>}
      localJointOffset=<3 0 0>
     }]
    relativeSpaceBone {boneName=root}
   }
   n6 = Anim Dynamics - Bone: dyn_sleeve_l_bk_01!(ComponentPose=@n5, Alpha=1, GravityOverride:Vector="0,0,-980", ExternalForce="0,0,0") -> (Pose) {
    angularDampingOverride=0.9999
    angularSpringConstant=500
    bAngularSpring=true
    bLinearSpring=true
    bOverrideAngularDamping=true
    bUseGravityOverride=true
    bUseSphericalLimits=true
    linearSpringConstant=50
    simulationSpace=BoneRelative
    alphaBoolBlend {blendOption=Linear}
    boundBone {boneName=dyn_sleeve_l_bk_01}
    componentAppliedLinearAccClamp=<1500 1500 0>
    componentLinearAccScale=<0.3 0.3 0.3>
    componentLinearVelScale=<0.3 0.3 0.3>
    gravityOverride {z=-980}
    physicsBodyDefinitions [{
      sphereCollisionRadius=0
      boundBone {boneName=dyn_sleeve_l_bk_01}
      boxExtents=<5 20 20>
      constraintSetup {angularLimitsMax=<2 60 30>, angularLimitsMin=<-2 -60 -30>, angularTarget=<0 0 0>}
      localJointOffset=<3 0 0>
     }]
    relativeSpaceBone {boneName=root}
   }
   n7 = Anim Dynamics - Bone: dyn_sleeve_l_bk_02!(ComponentPose=@n6, Alpha=1, GravityOverride:Vector="0,0,-980", ExternalForce="0,0,0") -> (Pose) {
    angularDampingOverride=0.9999
    angularSpringConstant=500
    bAngularSpring=true
    bLinearSpring=true
    bOverrideAngularDamping=true
    bUseGravityOverride=true
    bUseSphericalLimits=true
    linearSpringConstant=50
    simulationSpace=BoneRelative
    alphaBoolBlend {blendOption=Linear}
    boundBone {boneName=dyn_sleeve_l_bk_02}
    componentAppliedLinearAccClamp=<1500 1500 0>
    componentLinearAccScale=<0.3 0.3 0.3>
    componentLinearVelScale=<0.3 0.3 0.3>
    gravityOverride {z=-980}
    physicsBodyDefinitions [{
      sphereCollisionRadius=0
      boundBone {boneName=dyn_sleeve_l_bk_02}
      boxExtents=<5 20 20>
      constraintSetup {angularLimitsMax=<2 60 30>, angularLimitsMin=<-2 -60 -30>, angularTarget=<0 0 0>}
      localJointOffset=<3 0 0>
     }]
    relativeSpaceBone {boneName=root}
   }
   n8 = Anim Dynamics - Bone: dyn_sleeve_l_bk_03!(ComponentPose=@n7, Alpha=1, GravityOverride:Vector="0,0,-980", ExternalForce="0,0,0") -> (Pose) {
    angularDampingOverride=0.9999
    angularSpringConstant=500
    bAngularSpring=true
    bLinearSpring=true
    bOverrideAngularDamping=true
    bUseGravityOverride=true
    bUseSphericalLimits=true
    linearSpringConstant=50
    simulationSpace=BoneRelative
    alphaBoolBlend {blendOption=Linear}
    boundBone {boneName=dyn_sleeve_l_bk_03}
    componentAppliedLinearAccClamp=<1500 1500 0>
    componentLinearAccScale=<0.3 0.3 0.3>
    componentLinearVelScale=<0.3 0.3 0.3>
    gravityOverride {z=-980}
    physicsBodyDefinitions [{
      sphereCollisionRadius=0
      boundBone {boneName=dyn_sleeve_l_bk_03}
      boxExtents=<5 20 20>
      constraintSetup {angularLimitsMax=<2 60 30>, angularLimitsMin=<-2 -60 -30>, angularTarget=<0 0 0>}
      localJointOffset=<3 0 0>
     }]
    relativeSpaceBone {boneName=root}
   }
   n9 = Anim Dynamics - Bone: dyn_sleeve_l_bot_01!(ComponentPose=@n8, Alpha=1, GravityOverride:Vector="0,0,-980", ExternalForce="0,0,0") -> (Pose) {
    angularDampingOverride=0.9999
    angularSpringConstant=500
    bAngularSpring=true
    bLinearSpring=true
    bOverrideAngularDamping=true
    bUseGravityOverride=true
    bUseSphericalLimits=true
    linearSpringConstant=50
    simulationSpace=BoneRelative
    alphaBoolBlend {blendOption=Linear}
    boundBone {boneName=dyn_sleeve_l_bot_01}
    componentAppliedLinearAccClamp=<1500 1500 0>
    componentLinearAccScale=<0.3 0.3 0.3>
    componentLinearVelScale=<0.3 0.3 0.3>
    gravityOverride {z=-980}
    physicsBodyDefinitions [{
      sphereCollisionRadius=0
      boundBone {boneName=dyn_sleeve_l_bot_01}
      boxExtents=<5 20 20>
      constraintSetup {angularLimitsMax=<2 45 60>, angularLimitsMin=<-2 -45 -60>, angularTarget=<0 0 0>}
      localJointOffset=<3 0 0>
     }]
    relativeSpaceBone {boneName=root}
   }
   n10 = Anim Dynamics - Bone: dyn_sleeve_l_bot_02!(ComponentPose=@n9, Alpha=1, GravityOverride:Vector="0,0,-980", ExternalForce="0,0,0") -> (Pose) {
    angularDampingOverride=0.9999
    angularSpringConstant=500
    bAngularSpring=true
    bLinearSpring=true
    bOverrideAngularDamping=true
    bUseGravityOverride=true
    bUseSphericalLimits=true
    linearSpringConstant=50
    simulationSpace=BoneRelative
    alphaBoolBlend {blendOption=Linear}
    boundBone {boneName=dyn_sleeve_l_bot_02}
    componentAppliedLinearAccClamp=<1500 1500 0>
    componentLinearAccScale=<0.3 0.3 0.3>
    componentLinearVelScale=<0.3 0.3 0.3>
    gravityOverride {z=-980}
    physicsBodyDefinitions [{
      sphereCollisionRadius=0
      boundBone {boneName=dyn_sleeve_l_bot_02}
      boxExtents=<5 20 20>
      constraintSetup {angularLimitsMax=<2 45 60>, angularLimitsMin=<-2 -45 -60>, angularTarget=<0 0 0>}
      localJointOffset=<3 0 0>
     }]
    relativeSpaceBone {boneName=root}
   }
   n11 = Anim Dynamics - Bone: dyn_sleeve_l_bot_03!(ComponentPose=@n10, Alpha=1, GravityOverride:Vector="0,0,-980", ExternalForce="0,0,0") -> (Pose) {
    angularDampingOverride=0.9999
    angularSpringConstant=500
    bAngularSpring=true
    bLinearSpring=true
    bOverrideAngularDamping=true
    bUseGravityOverride=true
    bUseSphericalLimits=true
    linearSpringConstant=50
    simulationSpace=BoneRelative
    alphaBoolBlend {blendOption=Linear}
    boundBone {boneName=dyn_sleeve_l_bot_03}
    componentAppliedLinearAccClamp=<1500 1500 0>
    componentLinearAccScale=<0.3 0.3 0.3>
    componentLinearVelScale=<0.3 0.3 0.3>
    gravityOverride {z=-980}
    physicsBodyDefinitions [{
      sphereCollisionRadius=0
      boundBone {boneName=dyn_sleeve_l_bot_03}
      boxExtents=<5 20 20>
      constraintSetup {angularLimitsMax=<2 45 60>, angularLimitsMin=<-2 -45 -60>, angularTarget=<0 0 0>}
      localJointOffset=<3 0 0>
     }]
    relativeSpaceBone {boneName=root}
   }
   n12 = Anim Dynamics - Bone: dyn_sleeve_r_top_01!(ComponentPose=@n11, Alpha=1, GravityOverride:Vector="0,0,-980", ExternalForce="0,0,0") -> (Pose) {
    angularDampingOverride=0.9999
    angularSpringConstant=500
    bAngularSpring=true
    bLinearSpring=true
    bOverrideAngularDamping=true
    bUseGravityOverride=true
    bUseSphericalLimits=true
    linearSpringConstant=50
    simulationSpace=BoneRelative
    alphaBoolBlend {blendOption=Linear}
    boundBone {boneName=dyn_sleeve_r_top_01}
    componentAppliedLinearAccClamp=<1500 1500 0>
    componentLinearAccScale=<0.3 0.3 0.3>
    componentLinearVelScale=<0.3 0.3 0.3>
    gravityOverride {z=-980}
    physicsBodyDefinitions [{
      sphereCollisionRadius=0
      boundBone {boneName=dyn_sleeve_r_top_01}
      boxExtents=<5 20 20>
      constraintSetup {angularLimitsMax=<2 1 45>, angularLimitsMin=<-2 -30 -45>, angularTarget=<0 0 0>}
      localJointOffset=<-3 0 0>
     }]
    relativeSpaceBone {boneName=root}
   }
   n13 = Anim Dynamics - Bone: dyn_sleeve_r_top_02!(ComponentPose=@n12, Alpha=1, GravityOverride:Vector="0,0,-980", ExternalForce="0,0,0") -> (Pose) {
    angularDampingOverride=0.9999
    angularSpringConstant=500
    bAngularSpring=true
    bLinearSpring=true
    bOverrideAngularDamping=true
    bUseGravityOverride=true
    bUseSphericalLimits=true
    linearSpringConstant=50
    simulationSpace=BoneRelative
    alphaBoolBlend {blendOption=Linear}
    boundBone {boneName=dyn_sleeve_r_top_02}
    componentAppliedLinearAccClamp=<1500 1500 0>
    componentLinearAccScale=<0.3 0.3 0.3>
    componentLinearVelScale=<0.3 0.3 0.3>
    gravityOverride {z=-980}
    physicsBodyDefinitions [{
      sphereCollisionRadius=0
      boundBone {boneName=dyn_sleeve_r_top_02}
      boxExtents=<5 20 20>
      constraintSetup {angularLimitsMax=<2 1 45>, angularLimitsMin=<-2 -30 -45>, angularTarget=<0 0 0>}
      localJointOffset=<-3 0 0>
     }]
    relativeSpaceBone {boneName=root}
   }
   n14 = Anim Dynamics - Bone: dyn_sleeve_r_top_03!(ComponentPose=@n13, Alpha=1, GravityOverride:Vector="0,0,-980", ExternalForce="0,0,0") -> (Pose) {
    angularDampingOverride=0.9999
    angularSpringConstant=500
    bAngularSpring=true
    bLinearSpring=true
    bOverrideAngularDamping=true
    bUseGravityOverride=true
    bUseSphericalLimits=true
    linearSpringConstant=50
    simulationSpace=BoneRelative
    alphaBoolBlend {blendOption=Linear}
    boundBone {boneName=dyn_sleeve_r_top_03}
    componentAppliedLinearAccClamp=<1500 1500 0>
    componentLinearAccScale=<0.3 0.3 0.3>
    componentLinearVelScale=<0.3 0.3 0.3>
    gravityOverride {z=-980}
    physicsBodyDefinitions [{
      sphereCollisionRadius=0
      boundBone {boneName=dyn_sleeve_r_top_03}
      boxExtents=<5 20 20>
      constraintSetup {angularLimitsMax=<2 1 45>, angularLimitsMin=<-2 -30 -45>, angularTarget=<0 0 0>}
      localJointOffset=<-3 0 0>
     }]
    relativeSpaceBone {boneName=root}
   }
   n15 = Anim Dynamics - Bone: dyn_sleeve_r_bk_01!(ComponentPose=@n14, Alpha=1, GravityOverride:Vector="0,0,-980", ExternalForce="0,0,0") -> (Pose) {
    angularDampingOverride=0.9999
    angularSpringConstant=500
    bAngularSpring=true
    bLinearSpring=true
    bOverrideAngularDamping=true
    bUseGravityOverride=true
    bUseSphericalLimits=true
    linearSpringConstant=50
    simulationSpace=BoneRelative
    alphaBoolBlend {blendOption=Linear}
    boundBone {boneName=dyn_sleeve_r_bk_01}
    componentAppliedLinearAccClamp=<1500 1500 0>
    componentLinearAccScale=<0.3 0.3 0.3>
    componentLinearVelScale=<0.3 0.3 0.3>
    gravityOverride {z=-980}
    physicsBodyDefinitions [{
      sphereCollisionRadius=0
      boundBone {boneName=dyn_sleeve_r_bk_01}
      boxExtents=<5 20 20>
      constraintSetup {angularLimitsMax=<2 60 60>, angularLimitsMin=<-2 -60 -5>, angularTarget=<0 0 0>}
      localJointOffset=<-3 0 0>
     }]
    relativeSpaceBone {boneName=root}
   }
   n16 = Anim Dynamics - Bone: dyn_sleeve_r_bk_02!(ComponentPose=@n15, Alpha=1, GravityOverride:Vector="0,0,-980", ExternalForce="0,0,0") -> (Pose) {
    angularDampingOverride=0.9999
    angularSpringConstant=500
    bAngularSpring=true
    bLinearSpring=true
    bOverrideAngularDamping=true
    bUseGravityOverride=true
    bUseSphericalLimits=true
    linearSpringConstant=50
    simulationSpace=BoneRelative
    alphaBoolBlend {blendOption=Linear}
    boundBone {boneName=dyn_sleeve_r_bk_02}
    componentAppliedLinearAccClamp=<1500 1500 0>
    componentLinearAccScale=<0.3 0.3 0.3>
    componentLinearVelScale=<0.3 0.3 0.3>
    gravityOverride {z=-980}
    physicsBodyDefinitions [{
      sphereCollisionRadius=0
      boundBone {boneName=dyn_sleeve_r_bk_02}
      boxExtents=<5 20 20>
      constraintSetup {angularLimitsMax=<2 60 60>, angularLimitsMin=<-2 -60 -5>, angularTarget=<0 0 0>}
      localJointOffset=<-3 0 0>
     }]
    relativeSpaceBone {boneName=root}
   }
   n17 = Anim Dynamics - Bone: dyn_sleeve_r_bk_03!(ComponentPose=@n16, Alpha=1, GravityOverride:Vector="0,0,-980", ExternalForce="0,0,0") -> (Pose) {
    angularDampingOverride=0.9999
    angularSpringConstant=500
    bAngularSpring=true
    bLinearSpring=true
    bOverrideAngularDamping=true
    bUseGravityOverride=true
    bUseSphericalLimits=true
    linearSpringConstant=50
    simulationSpace=BoneRelative
    alphaBoolBlend {blendOption=Linear}
    boundBone {boneName=dyn_sleeve_r_bk_03}
    componentAppliedLinearAccClamp=<1500 1500 0>
    componentLinearAccScale=<0.3 0.3 0.3>
    componentLinearVelScale=<0.3 0.3 0.3>
    gravityOverride {z=-980}
    physicsBodyDefinitions [{
      sphereCollisionRadius=0
      boundBone {boneName=dyn_sleeve_r_bk_03}
      boxExtents=<5 20 20>
      constraintSetup {angularLimitsMax=<2 60 60>, angularLimitsMin=<-2 -60 -5>, angularTarget=<0 0 0>}
      localJointOffset=<-3 0 0>
     }]
    relativeSpaceBone {boneName=root}
   }
   n18 = Anim Dynamics - Bone: dyn_sleeve_r_bot_01!(ComponentPose=@n17, Alpha=1, GravityOverride:Vector="0,0,-980", ExternalForce="0,0,0") -> (Pose) {
    angularDampingOverride=0.9999
    angularSpringConstant=500
    bAngularSpring=true
    bLinearSpring=true
    bOverrideAngularDamping=true
    bUseGravityOverride=true
    bUseSphericalLimits=true
    linearSpringConstant=50
    simulationSpace=BoneRelative
    alphaBoolBlend {blendOption=Linear}
    boundBone {boneName=dyn_sleeve_r_bot_01}
    componentAppliedLinearAccClamp=<1500 1500 0>
    componentLinearAccScale=<0.3 0.3 0.3>
    componentLinearVelScale=<0.3 0.3 0.3>
    gravityOverride {z=-980}
    physicsBodyDefinitions [{
      sphereCollisionRadius=0
      boundBone {boneName=dyn_sleeve_r_bot_01}
      boxExtents=<5 20 20>
      constraintSetup {angularLimitsMax=<2 45 45>, angularLimitsMin=<-2 -45 -45>, angularTarget=<0 0 0>}
      localJointOffset=<-3 0 0>
     }]
    relativeSpaceBone {boneName=root}
   }
   n19 = Anim Dynamics - Bone: dyn_sleeve_r_bot_02!(ComponentPose=@n18, Alpha=1, GravityOverride:Vector="0,0,-980", ExternalForce="0,0,0") -> (Pose) {
    angularDampingOverride=0.9999
    angularSpringConstant=500
    bAngularSpring=true
    bLinearSpring=true
    bOverrideAngularDamping=true
    bUseGravityOverride=true
    bUseSphericalLimits=true
    linearSpringConstant=50
    simulationSpace=BoneRelative
    alphaBoolBlend {blendOption=Linear}
    boundBone {boneName=dyn_sleeve_r_bot_02}
    componentAppliedLinearAccClamp=<1500 1500 0>
    componentLinearAccScale=<0.3 0.3 0.3>
    componentLinearVelScale=<0.3 0.3 0.3>
    gravityOverride {z=-980}
    physicsBodyDefinitions [{
      sphereCollisionRadius=0
      boundBone {boneName=dyn_sleeve_r_bot_02}
      boxExtents=<5 20 20>
      constraintSetup {angularLimitsMax=<2 45 45>, angularLimitsMin=<-2 -45 -45>, angularTarget=<0 0 0>}
      localJointOffset=<-3 0 0>
     }]
    relativeSpaceBone {boneName=root}
   }
   n20 = Anim Dynamics - Bone: dyn_sleeve_r_bot_03!(ComponentPose=@n19, Alpha=1, GravityOverride:Vector="0,0,-980", ExternalForce="0,0,0") -> (Pose) {
    angularDampingOverride=0.9999
    angularSpringConstant=500
    bAngularSpring=true
    bLinearSpring=true
    bOverrideAngularDamping=true
    bUseGravityOverride=true
    bUseSphericalLimits=true
    linearSpringConstant=50
    simulationSpace=BoneRelative
    alphaBoolBlend {blendOption=Linear}
    boundBone {boneName=dyn_sleeve_r_bot_03}
    componentAppliedLinearAccClamp=<1500 1500 0>
    componentLinearAccScale=<0.3 0.3 0.3>
    componentLinearVelScale=<0.3 0.3 0.3>
    gravityOverride {z=-980}
    physicsBodyDefinitions [{
      sphereCollisionRadius=0
      boundBone {boneName=dyn_sleeve_r_bot_03}
      boxExtents=<5 20 20>
      constraintSetup {angularLimitsMax=<2 45 45>, angularLimitsMin=<-2 -45 -45>, angularTarget=<0 0 0>}
      localJointOffset=<-3 0 0>
     }]
    relativeSpaceBone {boneName=root}
   }
   n21 = Component To Local(ComponentPose=@n20)
   n22 = "Constraint - dyn_sleeve_r_top_01 (P)"(ComponentPose=Local To Component(LocalPose=Use cached pose 'StartupCache'()), Alpha=1, ConstraintWeights_0=1) {
    alphaBoolBlend {blendOption=Linear}
    alphaScaleBiasClamp {bClampResult=true, bInterpResult=true, interpSpeedDecreasing=5, interpSpeedIncreasing=7}
    boneToModify {boneName=dyn_sleeve_r_top_01}
    constraintSetup [{targetBone {boneName=lowerarm_r}}]
    constraintWeights [1]
   }
   n23 = "Constraint - dyn_sleeve_r_top_02 (P)"(ComponentPose=@n22, Alpha=1, ConstraintWeights_0=1) {
    alphaBoolBlend {blendOption=Linear}
    alphaScaleBiasClamp {bClampResult=true, bInterpResult=true, interpSpeedDecreasing=5, interpSpeedIncreasing=7}
    boneToModify {boneName=dyn_sleeve_r_top_02}
    constraintSetup [{targetBone {boneName=lowerarm_r}}]
    constraintWeights [1]
   }
   n24 = "Constraint - dyn_sleeve_r_top_03 (P)"(ComponentPose=@n23, Alpha=1, ConstraintWeights_0=1) {
    alphaBoolBlend {blendOption=Linear}
    alphaScaleBiasClamp {bClampResult=true, bInterpResult=true, clampMax=0.5, interpSpeedDecreasing=5, interpSpeedIncreasing=7}
    boneToModify {boneName=dyn_sleeve_r_top_03}
    constraintSetup [{targetBone {boneName=lowerarm_r}}]
    constraintWeights [1]
   }
   n25 = "Constraint - dyn_sleeve_r_bk_01 (P)"(ComponentPose=@n24, Alpha=1, ConstraintWeights_0=1) {
    alphaBoolBlend {blendOption=Linear}
    alphaScaleBiasClamp {bClampResult=true, bInterpResult=true, interpSpeedDecreasing=5, interpSpeedIncreasing=7}
    boneToModify {boneName=dyn_sleeve_r_bk_01}
    constraintSetup [{targetBone {boneName=lowerarm_r}}]
    constraintWeights [1]
   }
   n26 = "Constraint - dyn_sleeve_r_bk_02 (P)"(ComponentPose=@n25, Alpha=1, ConstraintWeights_0=1) {
    alphaBoolBlend {blendOption=Linear}
    alphaScaleBiasClamp {bClampResult=true, bInterpResult=true, interpSpeedDecreasing=5, interpSpeedIncreasing=7}
    boneToModify {boneName=dyn_sleeve_r_bk_02}
    constraintSetup [{targetBone {boneName=lowerarm_r}}]
    constraintWeights [1]
   }
   n27 = "Constraint - dyn_sleeve_r_bk_03 (P)"(ComponentPose=@n26, Alpha=1, ConstraintWeights_0=1) {
    alphaBoolBlend {blendOption=Linear}
    alphaScaleBiasClamp {bClampResult=true, bInterpResult=true, clampMax=0.5, interpSpeedDecreasing=5, interpSpeedIncreasing=7}
    boneToModify {boneName=dyn_sleeve_r_bk_03}
    constraintSetup [{targetBone {boneName=lowerarm_r}}]
    constraintWeights [1]
   }
   n28 = "Constraint - dyn_sleeve_r_bot_01 (P)"(ComponentPose=@n27, Alpha=1, ConstraintWeights_0=1) {
    alphaBoolBlend {blendOption=Linear}
    alphaScaleBiasClamp {bClampResult=true, bInterpResult=true, interpSpeedDecreasing=5, interpSpeedIncreasing=7}
    boneToModify {boneName=dyn_sleeve_r_bot_01}
    constraintSetup [{targetBone {boneName=lowerarm_r}}]
    constraintWeights [1]
   }
   n29 = "Constraint - dyn_sleeve_r_bot_02 (P)"(ComponentPose=@n28, Alpha=1, ConstraintWeights_0=1) {
    alphaBoolBlend {blendOption=Linear}
    alphaScaleBiasClamp {bClampResult=true, bInterpResult=true, interpSpeedDecreasing=5, interpSpeedIncreasing=7}
    boneToModify {boneName=dyn_sleeve_r_bot_02}
    constraintSetup [{targetBone {boneName=lowerarm_r}}]
    constraintWeights [1]
   }
   n30 = "Constraint - dyn_sleeve_r_bot_03 (P)"(ComponentPose=@n29, Alpha=1, ConstraintWeights_0=1) {
    alphaBoolBlend {blendOption=Linear}
    alphaScaleBiasClamp {bClampResult=true, bInterpResult=true, clampMax=0.5, interpSpeedDecreasing=5, interpSpeedIncreasing=7}
    boneToModify {boneName=dyn_sleeve_r_bot_03}
    constraintSetup [{targetBone {boneName=lowerarm_r}}]
    constraintWeights [1]
   }
   n31 = "Constraint - dyn_sleeve_l_top_01 (P)"(ComponentPose=@n30, Alpha=1, ConstraintWeights_0=1) {
    alphaBoolBlend {blendOption=Linear}
    alphaScaleBiasClamp {bClampResult=true, bInterpResult=true, interpSpeedDecreasing=5, interpSpeedIncreasing=7}
    boneToModify {boneName=dyn_sleeve_l_top_01}
    constraintSetup [{targetBone {boneName=lowerarm_l}}]
    constraintWeights [1]
   }
   n32 = "Constraint - dyn_sleeve_l_top_02 (P)"(ComponentPose=@n31, Alpha=1, ConstraintWeights_0=1) {
    alphaBoolBlend {blendOption=Linear}
    alphaScaleBiasClamp {bClampResult=true, bInterpResult=true, interpSpeedDecreasing=5, interpSpeedIncreasing=7}
    boneToModify {boneName=dyn_sleeve_l_top_02}
    constraintSetup [{targetBone {boneName=lowerarm_l}}]
    constraintWeights [1]
   }
   n33 = "Constraint - dyn_sleeve_l_top_03 (P)"(ComponentPose=@n32, Alpha=1, ConstraintWeights_0=1) {
    alphaBoolBlend {blendOption=Linear}
    alphaScaleBiasClamp {bClampResult=true, bInterpResult=true, clampMax=0.5, interpSpeedDecreasing=5, interpSpeedIncreasing=7}
    boneToModify {boneName=dyn_sleeve_l_top_03}
    constraintSetup [{targetBone {boneName=lowerarm_l}}]
    constraintWeights [1]
   }
   n34 = "Constraint - dyn_sleeve_l_bot_01 (P)"(ComponentPose=@n33, Alpha=1, ConstraintWeights_0=1) {
    alphaBoolBlend {blendOption=Linear}
    alphaScaleBiasClamp {bClampResult=true, bInterpResult=true, interpSpeedDecreasing=5, interpSpeedIncreasing=7}
    boneToModify {boneName=dyn_sleeve_l_bot_01}
    constraintSetup [{targetBone {boneName=lowerarm_l}}]
    constraintWeights [1]
   }
   n35 = "Constraint - dyn_sleeve_l_bot_02 (P)"(ComponentPose=@n34, Alpha=1, ConstraintWeights_0=1) {
    alphaBoolBlend {blendOption=Linear}
    alphaScaleBiasClamp {bClampResult=true, bInterpResult=true, interpSpeedDecreasing=5, interpSpeedIncreasing=7}
    boneToModify {boneName=dyn_sleeve_l_bot_02}
    constraintSetup [{targetBone {boneName=lowerarm_l}}]
    constraintWeights [1]
   }
   n36 = "Constraint - dyn_sleeve_l_bot_03 (P)"(ComponentPose=@n35, Alpha=1, ConstraintWeights_0=1) {
    alphaBoolBlend {blendOption=Linear}
    alphaScaleBiasClamp {bClampResult=true, bInterpResult=true, clampMax=0.5, interpSpeedDecreasing=5, interpSpeedIncreasing=7}
    boneToModify {boneName=dyn_sleeve_l_bot_03}
    constraintSetup [{targetBone {boneName=lowerarm_l}}]
    constraintWeights [1]
   }
   n37 = "Constraint - dyn_sleeve_l_bk_01 (P)"(ComponentPose=@n36, Alpha=1, ConstraintWeights_0=1) {
    alphaBoolBlend {blendOption=Linear}
    alphaScaleBiasClamp {bClampResult=true, bInterpResult=true, interpSpeedDecreasing=5, interpSpeedIncreasing=7}
    boneToModify {boneName=dyn_sleeve_l_bk_01}
    constraintSetup [{targetBone {boneName=lowerarm_l}}]
    constraintWeights [1]
   }
   n38 = "Constraint - dyn_sleeve_l_bk_02 (P)"(ComponentPose=@n37, Alpha=1, ConstraintWeights_0=1) {
    alphaBoolBlend {blendOption=Linear}
    alphaScaleBiasClamp {bClampResult=true, bInterpResult=true, interpSpeedDecreasing=5, interpSpeedIncreasing=7}
    boneToModify {boneName=dyn_sleeve_l_bk_02}
    constraintSetup [{targetBone {boneName=lowerarm_l}}]
    constraintWeights [1]
   }
   n39 = "Constraint - dyn_sleeve_l_bk_03 (P)"(ComponentPose=@n38, Alpha=1, ConstraintWeights_0=1) {
    alphaBoolBlend {blendOption=Linear}
    alphaScaleBiasClamp {bClampResult=true, bInterpResult=true, clampMax=0.5, interpSpeedDecreasing=5, interpSpeedIncreasing=7}
    boneToModify {boneName=dyn_sleeve_l_bk_03}
    constraintSetup [{targetBone {boneName=lowerarm_l}}]
    constraintWeights [1]
   }
   n40 = RigidBody(ComponentPose=@n39, OverrideWorldGravity="0,0,-980") {
    bOverrideWorldGravity=true
    overridePhysicsAsset=&F_MED_BlazerVeil_Physics
    simulationSpace=BaseBoneSpace
    alphaBoolBlend {blendOption=Linear}
    baseBoneRef {boneName=pelvis}
    componentAppliedLinearAccClamp=<6000 6000 6000>
    componentLinearAccScale=<0.3 1.2 1.2>
    componentLinearVelScale=<0.3 1.2 1.2>
    simSpaceSettings {
     maxAngularAcceleration=200
     maxAngularVelocity=5
     maxLinearAcceleration=1000
     maxLinearVelocity=200
     worldAlpha=0.55
     externalLinearDragV=<1 1>
    }
   }
   n41 = Component To Local(ComponentPose=@n40)
   n42 = Get bRigidBodyEnabled()
   n43 = "Constraint - dyn_sleeve_l_bot_01 (R)"(
    ComponentPose=Local To Component(LocalPose=Blend Poses by bool(BlendPose_0=@n41, BlendPose_1=@n21, BlendTime_0=0.35, BlendTime_1=0.35, bActiveValue=@n42) {blendType=Linear})
    Alpha=1
    ConstraintWeights_0=1
   ) {
    alphaBoolBlend {blendOption=Linear}
    alphaScaleBiasClamp {bClampResult=true, clampMax=0.15}
    boneToModify {boneName=dyn_sleeve_l_bot_01}
    constraintSetup [{transformType=Rotation, targetBone {boneName=dyn_sleeve_l_bk_01}}]
    constraintWeights [1]
   }
   n44 = "Constraint - dyn_sleeve_l_bot_02 (R)"(ComponentPose=@n43, Alpha=1, ConstraintWeights_0=1) {
    alphaBoolBlend {blendOption=Linear}
    alphaScaleBiasClamp {bClampResult=true, clampMax=0.15}
    boneToModify {boneName=dyn_sleeve_l_bot_02}
    constraintSetup [{transformType=Rotation, targetBone {boneName=dyn_sleeve_l_bk_02}}]
    constraintWeights [1]
   }
   n45 = "Constraint - dyn_sleeve_l_bot_03 (R)"(ComponentPose=@n44, Alpha=1, ConstraintWeights_0=1) {
    alphaBoolBlend {blendOption=Linear}
    alphaScaleBiasClamp {bClampResult=true, clampMax=0.15}
    boneToModify {boneName=dyn_sleeve_l_bot_03}
    constraintSetup [{transformType=Rotation, targetBone {boneName=dyn_sleeve_l_bk_03}}]
    constraintWeights [1]
   }
   n46 = "Constraint - dyn_sleeve_r_bot_01 (R)"(ComponentPose=@n45, Alpha=1, ConstraintWeights_0=1) {
    alphaBoolBlend {blendOption=Linear}
    alphaScaleBiasClamp {bClampResult=true, clampMax=0.15}
    boneToModify {boneName=dyn_sleeve_r_bot_01}
    constraintSetup [{transformType=Rotation, targetBone {boneName=dyn_sleeve_r_bk_01}}]
    constraintWeights [1]
   }
   n47 = "Constraint - dyn_sleeve_r_bot_02 (R)"(ComponentPose=@n46, Alpha=1, ConstraintWeights_0=1) {
    alphaBoolBlend {blendOption=Linear}
    alphaScaleBiasClamp {bClampResult=true, clampMax=0.15}
    boneToModify {boneName=dyn_sleeve_r_bot_02}
    constraintSetup [{transformType=Rotation, targetBone {boneName=dyn_sleeve_r_bk_02}}]
    constraintWeights [1]
   }
   n48 = "Constraint - dyn_sleeve_r_bot_03 (R)"(ComponentPose=@n47, Alpha=1, ConstraintWeights_0=1) {
    alphaBoolBlend {blendOption=Linear}
    alphaScaleBiasClamp {bClampResult=true, clampMax=0.15}
    boneToModify {boneName=dyn_sleeve_r_bot_03}
    constraintSetup [{transformType=Rotation, targetBone {boneName=dyn_sleeve_r_bk_03}}]
    constraintWeights [1]
   }
   n49 = "Constraint - dyn_sleeve_l_top_02 (RR)"!(ComponentPose=@n48, Alpha=1, ConstraintWeights_0=0.7, ConstraintWeights_1:float=0.3) -> (Pose) {
    alphaBoolBlend {blendOption=Linear}
    alphaScaleBiasClamp {bClampResult=true, clampMax=0.15}
    boneToModify {boneName=dyn_sleeve_l_top_02}
    constraintSetup [{transformType=Rotation, targetBone {boneName=dyn_sleeve_l_bk_02}}, {transformType=Rotation, targetBone {boneName=lowerarm_l}}]
    constraintWeights [1,1]
   }
   n50 = "Constraint - dyn_sleeve_l_top_03 (RR)"!(ComponentPose=@n49, Alpha=1, ConstraintWeights_0=0.7, ConstraintWeights_1:float=0.3) -> (Pose) {
    alphaBoolBlend {blendOption=Linear}
    alphaScaleBiasClamp {bClampResult=true, clampMax=0.15}
    boneToModify {boneName=dyn_sleeve_l_top_03}
    constraintSetup [{transformType=Rotation, targetBone {boneName=dyn_sleeve_l_bk_03}}, {transformType=Rotation, targetBone {boneName=lowerarm_l}}]
    constraintWeights [1,1]
   }
   n51 = "Constraint - dyn_sleeve_r_top_02 (RR)"!(ComponentPose=@n50, Alpha=1, ConstraintWeights_0=0.7, ConstraintWeights_1:float=0.3) -> (Pose) {
    alphaBoolBlend {blendOption=Linear}
    alphaScaleBiasClamp {bClampResult=true, clampMax=0.15}
    boneToModify {boneName=dyn_sleeve_r_top_02}
    constraintSetup [{transformType=Rotation, targetBone {boneName=dyn_sleeve_r_bk_02}}, {transformType=Rotation, targetBone {boneName=lowerarm_r}}]
    constraintWeights [1,1]
   }
   n52 = "Constraint - dyn_sleeve_r_top_03 (RR)"!(ComponentPose=@n51, Alpha=1, ConstraintWeights_0=0.7, ConstraintWeights_1:float=0.3) -> (Pose) {
    alphaBoolBlend {blendOption=Linear}
    alphaScaleBiasClamp {bClampResult=true, clampMax=0.15}
    boneToModify {boneName=dyn_sleeve_r_top_03}
    constraintSetup [{transformType=Rotation, targetBone {boneName=dyn_sleeve_r_bk_03}}, {transformType=Rotation, targetBone {boneName=lowerarm_r}}]
    constraintWeights [1,1]
   }
   n53 = "Constraint - dyn_sleeve_l_bk_01 (PP)"!(ComponentPose=@n52, Alpha=1, ConstraintWeights_0=0.8, ConstraintWeights_1:float=0.2) -> (Pose) {
    alphaBoolBlend {blendOption=Linear}
    alphaScaleBiasClamp {bClampResult=true, clampMax=0.3}
    boneToModify {boneName=dyn_sleeve_l_bk_01}
    constraintSetup [{targetBone {boneName=deform_elbow_out_l}}, {targetBone {boneName=lowerarm_l}}]
    constraintWeights [1,1]
   }
   n54 = "Constraint - dyn_sleeve_r_bk_01 (PP)"!(ComponentPose=@n53, Alpha=1, ConstraintWeights_0=0.8, ConstraintWeights_1:float=0.2) -> (Pose) {
    alphaBoolBlend {blendOption=Linear}
    alphaScaleBiasClamp {bClampResult=true, clampMax=0.3}
    boneToModify {boneName=dyn_sleeve_r_bk_01}
    constraintSetup [{targetBone {boneName=deform_elbow_out_r}}, {targetBone {boneName=lowerarm_r}}]
    constraintWeights [1,1]
   }
   n55 = "Constraint - dyn_sleeve_r_top_01 (P)"(ComponentPose=@n54, Alpha=0.5, ConstraintWeights_0=1) {
    alphaBoolBlend {blendOption=Linear}
    alphaScaleBiasClamp {bClampResult=true, interpSpeedDecreasing=5, interpSpeedIncreasing=7}
    boneToModify {boneName=dyn_sleeve_r_top_01}
    constraintSetup [{targetBone {boneName=lowerarm_r}}]
    constraintWeights [1]
   }
   n56 = "Constraint - dyn_sleeve_r_top_02 (PPPPPP)"!(
    ComponentPose=@n55
    Alpha=1
    ConstraintWeights_0=0.0081
    ConstraintWeights_1:float=0.0573
    ConstraintWeights_2:float=0.0175
    ConstraintWeights_3:float=0.1093
    ConstraintWeights_4:float=0.1467
    ConstraintWeights_5:float=0.6612
   ) -> (Pose) {
    alphaBoolBlend {blendOption=Linear}
    alphaScaleBiasClamp {bClampResult=true}
    boneToModify {boneName=dyn_sleeve_r_top_02}
    constraintSetup [
     {targetBone {boneName=upperarm_twist_01_r}}
     {targetBone {boneName=upperarm_r}}
     {targetBone {boneName=upperarm_twist_02_r}}
     {targetBone {boneName=lowerarm_twist_01_r}}
     {targetBone {boneName=lowerarm_r}}
     {targetBone {boneName=lowerarm_twist_02_r}}
    ]
    constraintWeights [1,1,1,1,1,1]
   }
   n57 = "Constraint - dyn_sleeve_r_top_03 (P)"(ComponentPose=@n56, Alpha=0.5, ConstraintWeights_0=1) {
    alphaBoolBlend {blendOption=Linear}
    alphaScaleBiasClamp {bClampResult=true, clampMax=0.01, interpSpeedDecreasing=5, interpSpeedIncreasing=7}
    boneToModify {boneName=dyn_sleeve_r_top_03}
    constraintSetup [{targetBone {boneName=lowerarm_r}}]
    constraintWeights [1]
   }
   n58 = "Constraint - dyn_sleeve_l_top_01 (P)"(ComponentPose=@n57, Alpha=0.5, ConstraintWeights_0=1) {
    alphaBoolBlend {blendOption=Linear}
    alphaScaleBiasClamp {bClampResult=true, interpSpeedDecreasing=5, interpSpeedIncreasing=7}
    boneToModify {boneName=dyn_sleeve_l_top_01}
    constraintSetup [{targetBone {boneName=lowerarm_l}}]
    constraintWeights [1]
   }
   n59 = "Constraint - dyn_sleeve_l_top_02 (PPPP)"!(
    ComponentPose=@n58
    Alpha=1
    ConstraintWeights_0=0.0196
    ConstraintWeights_1:float=0.1537
    ConstraintWeights_2:float=0.1172
    ConstraintWeights_3:float=0.7095
   ) -> (Pose) {
    alphaBoolBlend {blendOption=Linear}
    alphaScaleBiasClamp {bClampResult=true, clampMax=0.8}
    boneToModify {boneName=dyn_sleeve_l_top_02}
    constraintSetup [
     {targetBone {boneName=upperarm_l}}
     {targetBone {boneName=lowerarm_l}}
     {targetBone {boneName=lowerarm_twist_01_l}}
     {targetBone {boneName=lowerarm_twist_02_l}}
    ]
    constraintWeights [1,1,1,1]
   }
   n60 = "Constraint - dyn_sleeve_l_top_03 (P)"(ComponentPose=@n59, Alpha=0.5, ConstraintWeights_0=1) {
    alphaBoolBlend {blendOption=Linear}
    alphaScaleBiasClamp {bClampResult=true, clampMax=0.01, interpSpeedDecreasing=5, interpSpeedIncreasing=7}
    boneToModify {boneName=dyn_sleeve_l_top_03}
    constraintSetup [{targetBone {boneName=lowerarm_l}}]
    constraintWeights [1]
   }
   n61 = "Constraint - dyn_jacket_bk_01 (PPPPPPPP)"!(
    ComponentPose=@n60
    Alpha=1
    ConstraintWeights_0=0.1381
    ConstraintWeights_1:float=0.0566
    ConstraintWeights_2:float=0.0566
    ConstraintWeights_3:float=0.3354
    ConstraintWeights_4:float=0.1433
    ConstraintWeights_5:float=0.1973
    ConstraintWeights_6:float=0.0576
    ConstraintWeights_7:float=0.0151
   ) -> (Pose) {
    alphaBoolBlend {blendOption=Linear}
    boneToModify {boneName=dyn_jacket_bk_01}
    constraintSetup [
     {targetBone {boneName=spine_05}}
     {targetBone {boneName=clavicle_l}}
     {targetBone {boneName=clavicle_r}}
     {targetBone {boneName=spine_04}}
     {targetBone {boneName=spine_03}}
     {targetBone {boneName=spine_02}}
     {targetBone {boneName=spine_01}}
     {targetBone {boneName=pelvis}}
    ]
    constraintWeights [1,1,1,1,1,1,1,1]
   }
   n62 = "Constraint - dyn_jacket_l_01 (PPPPPPP)"!(
    ComponentPose=@n61
    Alpha=1
    ConstraintWeights_0=0.1342
    ConstraintWeights_1:float=0.12
    ConstraintWeights_2:float=0.5391
    ConstraintWeights_3:float=0.1225
    ConstraintWeights_4:float=0.0401
    ConstraintWeights_5:float=0.0408
    ConstraintWeights_6:float=0.0034
   ) -> (Pose) {
    alphaBoolBlend {blendOption=Linear}
    boneToModify {boneName=dyn_jacket_l_01}
    constraintSetup [
     {targetBone {boneName=spine_05}}
     {targetBone {boneName=clavicle_l}}
     {targetBone {boneName=spine_04}}
     {targetBone {boneName=upperarm_twist_01_l}}
     {targetBone {boneName=spine_03}}
     {targetBone {boneName=spine_02}}
     {targetBone {boneName=spine_01}}
    ]
    constraintWeights [1,1,1,1,1,1,1]
   }
   n63 = "Constraint - dyn_jacket_l_02 (PPPPPPPP)"!(
    ComponentPose=@n62
    Alpha=1
    ConstraintWeights_0=0.0584
    ConstraintWeights_1:float=0.0525
    ConstraintWeights_2:float=0.5268
    ConstraintWeights_3:float=0.0297
    ConstraintWeights_4:float=0.1058
    ConstraintWeights_5:float=0.1794
    ConstraintWeights_6:float=0.035
    ConstraintWeights_7:float=0.0124
   ) -> (Pose) {
    alphaBoolBlend {blendOption=Linear}
    boneToModify {boneName=dyn_jacket_l_02}
    constraintSetup [
     {targetBone {boneName=spine_05}}
     {targetBone {boneName=clavicle_l}}
     {targetBone {boneName=spine_04}}
     {targetBone {boneName=upperarm_twist_01_l}}
     {targetBone {boneName=spine_03}}
     {targetBone {boneName=spine_02}}
     {targetBone {boneName=spine_01}}
     {targetBone {boneName=pelvis}}
    ]
    constraintWeights [1,1,1,1,1,1,1,1]
   }
   n64 = "Constraint - dyn_jacket_r_01 (PPPPPPP)"!(
    ComponentPose=@n63
    Alpha=1
    ConstraintWeights_0=0.1536
    ConstraintWeights_1:float=0.1442
    ConstraintWeights_2:float=0.4719
    ConstraintWeights_3:float=0.1976
    ConstraintWeights_4:float=0.0174
    ConstraintWeights_5:float=0.0145
    ConstraintWeights_6:float=0.0007
   ) -> (Pose) {
    alphaBoolBlend {blendOption=Linear}
    boneToModify {boneName=dyn_jacket_r_01}
    constraintSetup [
     {targetBone {boneName=spine_05}}
     {targetBone {boneName=clavicle_r}}
     {targetBone {boneName=spine_04}}
     {targetBone {boneName=upperarm_twist_01_r}}
     {targetBone {boneName=spine_03}}
     {targetBone {boneName=spine_02}}
     {targetBone {boneName=deform_upperarm_bk_r}}
    ]
    constraintWeights [1,1,1,1,1,1,1]
   }
   n65 = "Constraint - dyn_jacket_r_02 (PPPPPPPP)"!(
    ComponentPose=@n64
    Alpha=1
    ConstraintWeights_0=0.0638
    ConstraintWeights_1:float=0.0614
    ConstraintWeights_2:float=0.5057
    ConstraintWeights_3:float=0.0352
    ConstraintWeights_4:float=0.108
    ConstraintWeights_5:float=0.1794
    ConstraintWeights_6:float=0.0364
    ConstraintWeights_7:float=0.0101
   ) -> (Pose) {
    alphaBoolBlend {blendOption=Linear}
    boneToModify {boneName=dyn_jacket_r_02}
    constraintSetup [
     {targetBone {boneName=spine_05}}
     {targetBone {boneName=clavicle_r}}
     {targetBone {boneName=spine_04}}
     {targetBone {boneName=upperarm_twist_01_r}}
     {targetBone {boneName=spine_03}}
     {targetBone {boneName=spine_02}}
     {targetBone {boneName=spine_01}}
     {targetBone {boneName=pelvis}}
    ]
    constraintWeights [1,1,1,1,1,1,1,1]
   }
   n66 = "Constraint - dyn_jacket_l_02 (PPPPPPPP)"(
    ComponentPose=@n65
    Alpha=1
    ConstraintWeights_0=0.0529
    ConstraintWeights_1=0.0436
    ConstraintWeights_2=0.5479
    ConstraintWeights_3=0.0243
    ConstraintWeights_4=0.1036
    ConstraintWeights_5=0.1795
    ConstraintWeights_6=0.0335
    ConstraintWeights_7=0.0146
   ) {
    alphaBoolBlend {blendOption=Linear}
    alphaScaleBiasClamp {bInterpResult=true, interpSpeedDecreasing=2, interpSpeedIncreasing=5}
    boneToModify {boneName=dyn_jacket_l_02}
    constraintSetup [
     {targetBone {boneName=spine_05}}
     {targetBone {boneName=clavicle_l}}
     {targetBone {boneName=spine_04}}
     {targetBone {boneName=upperarm_twist_01_l}}
     {targetBone {boneName=spine_03}}
     {targetBone {boneName=spine_02}}
     {targetBone {boneName=spine_01}}
     {targetBone {boneName=pelvis}}
    ]
    constraintWeights [1,1,1,1,1,1,1,1]
   }
   n67 = "Constraint - dyn_jacket_r_02 (PPPPPPPP)"(
    ComponentPose=@n66
    Alpha=1
    ConstraintWeights_0=0.0638
    ConstraintWeights_1=0.0614
    ConstraintWeights_2=0.5057
    ConstraintWeights_3=0.0352
    ConstraintWeights_4=0.108
    ConstraintWeights_5=0.1794
    ConstraintWeights_6=0.0364
    ConstraintWeights_7=0.0101
   ) {
    alphaBoolBlend {blendOption=Linear}
    alphaScaleBiasClamp {bInterpResult=true, interpSpeedDecreasing=2, interpSpeedIncreasing=5}
    boneToModify {boneName=dyn_jacket_r_02}
    constraintSetup [
     {targetBone {boneName=spine_05}}
     {targetBone {boneName=clavicle_r}}
     {targetBone {boneName=spine_04}}
     {targetBone {boneName=upperarm_twist_01_r}}
     {targetBone {boneName=spine_03}}
     {targetBone {boneName=spine_02}}
     {targetBone {boneName=spine_01}}
     {targetBone {boneName=pelvis}}
    ]
    constraintWeights [1,1,1,1,1,1,1,1]
   }
   n68 = "Constraint - dyn_jacket_bk_01 (PPPPPPPP)"(
    ComponentPose=@n67
    Alpha=1
    ConstraintWeights_0=0.1381
    ConstraintWeights_1=0.0566
    ConstraintWeights_2=0.0566
    ConstraintWeights_3=0.3354
    ConstraintWeights_4=0.1433
    ConstraintWeights_5=0.1973
    ConstraintWeights_6=0.0576
    ConstraintWeights_7=0.0151
   ) {
    alphaBoolBlend {blendOption=Linear}
    boneToModify {boneName=dyn_jacket_bk_01}
    constraintSetup [
     {targetBone {boneName=spine_05}}
     {targetBone {boneName=clavicle_l}}
     {targetBone {boneName=clavicle_r}}
     {targetBone {boneName=spine_04}}
     {targetBone {boneName=spine_03}}
     {targetBone {boneName=spine_02}}
     {targetBone {boneName=spine_01}}
     {targetBone {boneName=pelvis}}
    ]
    constraintWeights [1,1,1,1,1,1,1,1]
   }
   n69 = "Constraint - dyn_jacket_bk_02 (PPPPPP)"!(
    ComponentPose=@n68
    Alpha=1
    ConstraintWeights_0=0.0199
    ConstraintWeights_1:float=0.1569
    ConstraintWeights_2:float=0.1414
    ConstraintWeights_3:float=0.4547
    ConstraintWeights_4:float=0.106
    ConstraintWeights_5:float=0.1212
   ) -> (Pose) {
    alphaBoolBlend {blendOption=Linear}
    boneToModify {boneName=dyn_jacket_bk_02}
    constraintSetup [
     {targetBone {boneName=spine_05}}
     {targetBone {boneName=spine_04}}
     {targetBone {boneName=spine_03}}
     {targetBone {boneName=spine_02}}
     {targetBone {boneName=spine_01}}
     {targetBone {boneName=pelvis}}
    ]
    constraintWeights [1,1,1,1,1,1]
   }
   n70 = "Constraint - dyn_jacket_bk_03 (PPPPP)"!(
    ComponentPose=@n69
    Alpha=1
    ConstraintWeights_0=0.0257
    ConstraintWeights_1:float=0.034
    ConstraintWeights_2:float=0.2642
    ConstraintWeights_3:float=0.103
    ConstraintWeights_4:float=0.5731
   ) -> (Pose) {
    alphaBoolBlend {blendOption=Linear}
    boneToModify {boneName=dyn_jacket_bk_03}
    constraintSetup [
     {targetBone {boneName=spine_04}}
     {targetBone {boneName=spine_03}}
     {targetBone {boneName=spine_02}}
     {targetBone {boneName=spine_01}}
     {targetBone {boneName=pelvis}}
    ]
    constraintWeights [1,1,1,1,1]
   }
   n71 = "Constraint - dyn_jacket_l_02 (PPPPPPPP)"(
    ComponentPose=@n70
    Alpha=1
    ConstraintWeights_0=0.0529
    ConstraintWeights_1=0.0436
    ConstraintWeights_2=0.5479
    ConstraintWeights_3=0.0243
    ConstraintWeights_4=0.1036
    ConstraintWeights_5=0.1795
    ConstraintWeights_6=0.0335
    ConstraintWeights_7=0.0146
   ) {
    alphaBoolBlend {blendOption=Linear}
    boneToModify {boneName=dyn_jacket_l_02}
    constraintSetup [
     {targetBone {boneName=spine_05}}
     {targetBone {boneName=clavicle_l}}
     {targetBone {boneName=spine_04}}
     {targetBone {boneName=upperarm_twist_01_l}}
     {targetBone {boneName=spine_03}}
     {targetBone {boneName=spine_02}}
     {targetBone {boneName=spine_01}}
     {targetBone {boneName=pelvis}}
    ]
    constraintWeights [1,1,1,1,1,1,1,1]
   }
   n72 = "Constraint - dyn_jacket_r_02 (PPPPPPPP)"(
    ComponentPose=@n71
    Alpha=1
    ConstraintWeights_0=0.0638
    ConstraintWeights_1=0.0614
    ConstraintWeights_2=0.5057
    ConstraintWeights_3=0.0352
    ConstraintWeights_4=0.108
    ConstraintWeights_5=0.1794
    ConstraintWeights_6=0.0364
    ConstraintWeights_7=0.0101
   ) {
    alphaBoolBlend {blendOption=Linear}
    boneToModify {boneName=dyn_jacket_r_02}
    constraintSetup [
     {targetBone {boneName=spine_05}}
     {targetBone {boneName=clavicle_r}}
     {targetBone {boneName=spine_04}}
     {targetBone {boneName=upperarm_twist_01_r}}
     {targetBone {boneName=spine_03}}
     {targetBone {boneName=spine_02}}
     {targetBone {boneName=spine_01}}
     {targetBone {boneName=pelvis}}
    ]
    constraintWeights [1,1,1,1,1,1,1,1]
   }
   n73 = "Transform (Modify) Bone - Bone: dyn_tie_01"(ComponentPose=@n72, Alpha=1, Translation="0,0,0", Rotation="0,45,0", Scale="1,1,1") {
    rotationMode=BMM_Replace
    rotationSpace=BCS_BoneSpace
    alphaBoolBlend {blendOption=Linear}
    alphaScaleBiasClamp {bInterpResult=true, interpSpeedIncreasing=15}
    boneToModify {boneName=dyn_tie_01}
    rotation {yaw=45}
   }
   n74 = "Constraint - dyn_sleeve_r_bk_03 (P)"(ComponentPose=@n73, Alpha=1, ConstraintWeights_0=1) {
    alphaBoolBlend {blendOption=Linear}
    alphaScaleBiasClamp {bClampResult=true, bInterpResult=true, interpSpeedDecreasing=5, interpSpeedIncreasing=7}
    boneToModify {boneName=dyn_sleeve_r_bk_03}
    constraintSetup [{targetBone {boneName=lowerarm_r}}]
    constraintWeights [1]
   }
   n75 = "Constraint - dyn_sleeve_r_top_02 (P)"(ComponentPose=@n74, Alpha=1, ConstraintWeights_0=1) {
    alphaBoolBlend {blendOption=Linear}
    alphaScaleBiasClamp {bClampResult=true, bInterpResult=true, interpSpeedDecreasing=5, interpSpeedIncreasing=7}
    boneToModify {boneName=dyn_sleeve_r_top_02}
    constraintSetup [{targetBone {boneName=lowerarm_r}}]
    constraintWeights [1]
   }
   n76 = "Constraint - dyn_sleeve_r_bot_02 (P)"(ComponentPose=@n75, Alpha=1, ConstraintWeights_0=1) {
    alphaBoolBlend {blendOption=Linear}
    alphaScaleBiasClamp {bClampResult=true, bInterpResult=true, interpSpeedDecreasing=5, interpSpeedIncreasing=7}
    boneToModify {boneName=dyn_sleeve_r_bot_02}
    constraintSetup [{targetBone {boneName=lowerarm_r}}]
    constraintWeights [1]
   }
   n77 = "Transform (Modify) Bone - Bone: dyn_sleeve_r_top_MSTR"!(ComponentPose=@n76, Alpha=1, Rotation="4.903,-10.030,-0.065") -> (Pose) {
    rotationMode=BMM_Additive
    rotationSpace=BCS_BoneSpace
    alphaBoolBlend {blendOption=Linear}
    boneToModify {boneName=dyn_sleeve_r_top_MSTR}
    rotation=<rot 4.903 -10.03 -0.065>
   }
   n78 = "Transform (Modify) Bone - Bone: dyn_sleeve_r_bk_MSTR"!(ComponentPose=@n77, Alpha=1, Rotation="4.870,-15.048,-0.493") -> (Pose) {
    rotationMode=BMM_Additive
    rotationSpace=BCS_BoneSpace
    alphaBoolBlend {blendOption=Linear}
    boneToModify {boneName=dyn_sleeve_r_bk_MSTR}
    rotation=<rot 4.87 -15.048 -0.493>
   }
   n79 = "Constraint - dyn_sleeve_l_bot_01 (P)"(ComponentPose=@n78, Alpha=1, ConstraintWeights_0=1) {
    alphaBoolBlend {blendOption=Linear}
    alphaScaleBiasClamp {bClampResult=true, bInterpResult=true, interpSpeedDecreasing=5, interpSpeedIncreasing=7}
    boneToModify {boneName=dyn_sleeve_l_bot_01}
    constraintSetup [{targetBone {boneName=lowerarm_l}}]
    constraintWeights [1]
   }
   n80 = "Constraint - dyn_sleeve_l_bot_02 (P)"(ComponentPose=@n79, Alpha=1, ConstraintWeights_0=1) {
    alphaBoolBlend {blendOption=Linear}
    alphaScaleBiasClamp {bClampResult=true, bInterpResult=true, interpSpeedDecreasing=5, interpSpeedIncreasing=7}
    boneToModify {boneName=dyn_sleeve_l_bot_02}
    constraintSetup [{targetBone {boneName=lowerarm_l}}]
    constraintWeights [1]
   }
   n81 = "Constraint - dyn_sleeve_l_bot_03 (P)"(ComponentPose=@n80, Alpha=1, ConstraintWeights_0=1) {
    alphaBoolBlend {blendOption=Linear}
    alphaScaleBiasClamp {bClampResult=true, bInterpResult=true, interpSpeedDecreasing=5, interpSpeedIncreasing=7}
    boneToModify {boneName=dyn_sleeve_l_bot_03}
    constraintSetup [{targetBone {boneName=lowerarm_l}}]
    constraintWeights [1]
   }
   n82 = "Transform (Modify) Bone - Bone: dyn_tie_01"!(ComponentPose=@n81, Alpha=1, Rotation="0.005,3.000,-0.006") -> (Pose) {
    rotationMode=BMM_Additive
    rotationSpace=BCS_BoneSpace
    alphaBoolBlend {blendOption=Linear}
    boneToModify {boneName=dyn_tie_01}
    rotation=<rot 0.005 3 -0.006>
   }
   n83 = "Transform (Modify) Bone - Bone: dyn_tie_03"!(ComponentPose=@n82, Alpha=1, Rotation="-2,-7.743,0") -> (Pose) {
    rotationMode=BMM_Additive
    rotationSpace=BCS_BoneSpace
    alphaBoolBlend {blendOption=Linear}
    boneToModify {boneName=dyn_tie_03}
    rotation {pitch=-2, yaw=-7.743}
   }
   n84 = Layered blend per bone(
    BasePose=Component To Local(ComponentPose=@n83)
    BlendPoses_0=Slot 'DefaultSlot'()
    BlendWeights_0=Get Custom Idle Is Playing() -> (Custom Idle Is Playing:float)
   ) {layerSetup [{branchFilters [
       {boneName=dyn_sleeve_l_bk_MSTR}
       {boneName=dyn_sleeve_l_bot_MSTR}
       {boneName=dyn_sleeve_l_top_MSTR}
       {boneName=dyn_sleeve_r_bk_MSTR}
       {boneName=dyn_sleeve_r_bot_MSTR}
       {boneName=dyn_sleeve_r_top_MSTR}
       {boneName=dyn_collar_l_01}
       {boneName=dyn_collar_r_01}
       {boneName=dyn_jacket_l_01}
       {boneName=dyn_jacket_r_01}
      ]}]}
   Output Pose(Result=@n84) {layerGroup=None}
   n85 = Anim Dynamics - Bone: dyn_ankle_top!(
    ComponentPose=Local To Component(LocalPose=Copy Pose From Mesh() {bUseAttachedParent=true})
    Alpha=1
    GravityOverride:Vector="0,0,-980"
    ExternalForce="0,0,0"
   ) -> (Pose) {
    angularDampingOverride=0.99
    angularSpringConstant=100
    bAngularSpring=true
    bOverrideAngularDamping=true
    bUseGravityOverride=true
    bUseSphericalLimits=true
    simulationSpace=RootRelative
    alphaBoolBlend {blendOption=Linear}
    boundBone {boneName=dyn_ankle_top}
    componentAppliedLinearAccClamp=<2000 2000 2000>
    componentLinearAccScale=<0.8 0.8 0.8>
    componentLinearVelScale=<0.8 0.8 0.8>
    gravityOverride {z=-980}
    physicsBodyDefinitions [{
      collisionType=CustomSphere
      sphereCollisionRadius=4.5
      boundBone {boneName=dyn_ankle_top}
      boxExtents=<5 25 25>
      constraintSetup {angularLimitsMax=<1 20 20>, angularLimitsMin=<-1 -20 -20>, angularTarget=<1 0 0>}
      localJointOffset=<3 0 0>
     }]
    relativeSpaceBone {boneName=pelvis}
   }
   n86 = Anim Dynamics - Bone: dyn_ankle_bot!(ComponentPose=@n85, Alpha=1, GravityOverride:Vector="0,0,-980", ExternalForce="0,0,0") -> (Pose) {
    angularDampingOverride=0.99
    angularSpringConstant=100
    bAngularSpring=true
    bOverrideAngularDamping=true
    bUseGravityOverride=true
    bUseSphericalLimits=true
    simulationSpace=RootRelative
    alphaBoolBlend {blendOption=Linear}
    boundBone {boneName=dyn_ankle_bot}
    componentAppliedLinearAccClamp=<2000 2000 2000>
    componentLinearAccScale=<0.8 0.8 0.8>
    componentLinearVelScale=<0.8 0.8 0.8>
    gravityOverride {z=-980}
    physicsBodyDefinitions [{
      collisionType=CustomSphere
      sphereCollisionRadius=4.5
      boundBone {boneName=dyn_ankle_bot}
      boxExtents=<5 25 25>
      constraintSetup {angularLimitsMax=<1 20 20>, angularLimitsMin=<-1 -20 -20>, angularTarget=<1 0 0>}
      localJointOffset=<3 0 0>
     }]
    relativeSpaceBone {boneName=pelvis}
   }
   n87 = "Constraint - dyn_tie_01 (PPPPPP)"!(
    ComponentPose=@n86
    Alpha=0.6
    ConstraintWeights_0=0.128
    ConstraintWeights_1:float=0.8503
    ConstraintWeights_2:float=0.006
    ConstraintWeights_3:float=0.0053
    ConstraintWeights_4:float=0.0104
    ConstraintWeights_5:float=0
   ) -> (Pose) {
    alphaBoolBlend {blendOption=Linear}
    boneToModify {boneName=dyn_tie_01}
    constraintSetup [
     {targetBone {boneName=spine_05}}
     {targetBone {boneName=spine_04}}
     {targetBone {boneName=spine_03}}
     {targetBone {boneName=spine_02}}
     {targetBone {boneName=deform_pec_l}}
     {targetBone {boneName=deform_pec_r}}
    ]
    constraintWeights [1,1,1,1,1,1]
   }
   n88 = Anim Dynamics - Bone: dyn_tie_01!(ComponentPose=@n87, Alpha=1, GravityOverride:Vector="0,0,-980", ExternalForce="0,0,0") -> (Pose) {
    angularDampingOverride=0.99
    angularSpringConstant=100
    bAngularSpring=true
    bOverrideAngularDamping=true
    bUseGravityOverride=true
    bUseSphericalLimits=true
    simulationSpace=RootRelative
    alphaBoolBlend {blendOption=Linear}
    boundBone {boneName=dyn_tie_01}
    componentAppliedLinearAccClamp=<2000 2000 2000>
    componentLinearAccScale=<0.8 0.8 0.8>
    componentLinearVelScale=<0.8 0.8 0.8>
    gravityOverride {z=-980}
    physicsBodyDefinitions [{
      collisionType=CustomSphere
      sphereCollisionRadius=4.5
      boundBone {boneName=dyn_tie_01}
      boxExtents=<25 15 15>
      constraintSetup {angularLimitsMax=<1 20 20>, angularLimitsMin=<-1 -20 -20>, angularTarget=<1 0 0>}
      localJointOffset=<2 0 0>
     }]
    relativeSpaceBone {boneName=pelvis}
    sphericalLimits [
     {limitRadius=8, drivingBone {boneName=spine_04}, sphereLocalOffset=<0 0 0>}
     {limitRadius=8, drivingBone {boneName=spine_05}, sphereLocalOffset=<0 0 0>}
    ]
   }
   n89 = Anim Dynamics - Bone: dyn_tie_02!(ComponentPose=@n88, Alpha=1, GravityOverride:Vector="0,0,-980", ExternalForce="0,0,0") -> (Pose) {
    angularDampingOverride=0.99
    angularSpringConstant=100
    bAngularSpring=true
    bOverrideAngularDamping=true
    bUseGravityOverride=true
    bUseSphericalLimits=true
    simulationSpace=RootRelative
    alphaBoolBlend {blendOption=Linear}
    boundBone {boneName=dyn_tie_02}
    componentAppliedLinearAccClamp=<2000 2000 2000>
    componentLinearAccScale=<0.8 0.8 0.8>
    componentLinearVelScale=<0.8 0.8 0.8>
    gravityOverride {z=-980}
    physicsBodyDefinitions [{
      collisionType=CustomSphere
      sphereCollisionRadius=4.5
      boundBone {boneName=dyn_tie_02}
      boxExtents=<25 15 15>
      constraintSetup {angularLimitsMax=<1 20 20>, angularLimitsMin=<-1 -20 -20>, angularTarget=<1 0 0>}
      localJointOffset=<2 0 0>
     }]
    relativeSpaceBone {boneName=pelvis}
    sphericalLimits [
     {limitRadius=8, drivingBone {boneName=spine_04}, sphereLocalOffset=<0 0 0>}
     {limitRadius=8, drivingBone {boneName=spine_05}, sphereLocalOffset=<0 0 0>}
    ]
   }
   n90 = Anim Dynamics - Bone: dyn_tie_03!(ComponentPose=@n89, Alpha=1, GravityOverride:Vector="0,0,-980", ExternalForce="0,0,0") -> (Pose) {
    angularDampingOverride=0.99
    angularSpringConstant=100
    bAngularSpring=true
    bOverrideAngularDamping=true
    bUseGravityOverride=true
    bUseSphericalLimits=true
    simulationSpace=RootRelative
    alphaBoolBlend {blendOption=Linear}
    boundBone {boneName=dyn_tie_03}
    componentAppliedLinearAccClamp=<2000 2000 2000>
    componentLinearAccScale=<0.8 0.8 0.8>
    componentLinearVelScale=<0.8 0.8 0.8>
    gravityOverride {z=-980}
    physicsBodyDefinitions [{
      collisionType=CustomSphere
      sphereCollisionRadius=3.5
      boundBone {boneName=dyn_tie_03}
      boxExtents=<25 15 15>
      constraintSetup {angularLimitsMax=<1 20 20>, angularLimitsMin=<-1 -20 -20>, angularTarget=<1 0 0>}
      localJointOffset=<2 0 0>
     }]
    relativeSpaceBone {boneName=pelvis}
    sphericalLimits [
     {limitRadius=8, drivingBone {boneName=spine_02}, sphereLocalOffset=<0 0 0>}
     {limitRadius=8, drivingBone {boneName=spine_03}, sphereLocalOffset=<0 0 0>}
    ]
   }
   n91 = Anim Dynamics - Bone: dyn_tie_04!(ComponentPose=@n90, Alpha=1, GravityOverride:Vector="0,0,-980", ExternalForce="0,0,0") -> (Pose) {
    angularDampingOverride=0.99
    angularSpringConstant=100
    bAngularSpring=true
    bOverrideAngularDamping=true
    bUseGravityOverride=true
    bUseSphericalLimits=true
    simulationSpace=RootRelative
    alphaBoolBlend {blendOption=Linear}
    boundBone {boneName=dyn_tie_04}
    componentAppliedLinearAccClamp=<2000 2000 2000>
    componentLinearAccScale=<0.8 0.8 0.8>
    componentLinearVelScale=<0.8 0.8 0.8>
    gravityOverride {z=-980}
    physicsBodyDefinitions [{
      collisionType=CustomSphere
      sphereCollisionRadius=3.5
      boundBone {boneName=dyn_tie_04}
      boxExtents=<25 15 15>
      constraintSetup {angularLimitsMax=<1 20 20>, angularLimitsMin=<-1 -20 -20>, angularTarget=<1 0 0>}
      localJointOffset=<2 0 0>
     }]
    relativeSpaceBone {boneName=pelvis}
    sphericalLimits [
     {limitRadius=8, drivingBone {boneName=spine_02}, sphereLocalOffset=<0 0 0>}
     {limitRadius=8, drivingBone {boneName=spine_03}, sphereLocalOffset=<0 0 0>}
    ]
   }
   n92 = Anim Dynamics - Bone: dyn_collar_r_01(ComponentPose=@n91, Alpha=1, ExternalForce="0,0,0") {
    angularDampingOverride=1
    angularSpringConstant=20000
    bAngularSpring=true
    bOverrideAngularDamping=true
    bUseGravityOverride=true
    gravityScale=0.35
    simulationSpace=RootRelative
    alphaBoolBlend {blendOption=Linear}
    boundBone {boneName=dyn_collar_r_01}
    componentLinearAccScale=<0.7 0.7 0.7>
    componentLinearVelScale=<0.7 0.7 0.7>
    gravityOverride {z=-980}
    physicsBodyDefinitions [{
      sphereCollisionRadius=0
      boundBone {boneName=dyn_collar_r_01}
      boxExtents=<20 30 30>
      constraintSetup {angularLimitsMax=<0 5 10>, angularLimitsMin=<0 -2 -10>, angularTarget=<1 0 0>}
      localJointOffset=<4 0 0>
     }]
   }
   n93 = Anim Dynamics - Bone: dyn_collar_r_02(ComponentPose=@n92, Alpha=1, ExternalForce="0,0,0") {
    angularDampingOverride=1
    angularSpringConstant=20000
    bAngularSpring=true
    bOverrideAngularDamping=true
    bUseGravityOverride=true
    gravityScale=0.35
    simulationSpace=RootRelative
    alphaBoolBlend {blendOption=Linear}
    boundBone {boneName=dyn_collar_r_02}
    componentLinearAccScale=<0.7 0.7 0.7>
    componentLinearVelScale=<0.7 0.7 0.7>
    gravityOverride {z=-980}
    physicsBodyDefinitions [{
      sphereCollisionRadius=0
      boundBone {boneName=dyn_collar_r_02}
      boxExtents=<20 30 30>
      constraintSetup {angularLimitsMax=<0 5 10>, angularLimitsMin=<0 -2 -10>, angularTarget=<1 0 0>}
      localJointOffset=<4 0 0>
     }]
   }
   n94 = Anim Dynamics - Bone: dyn_collar_l_01(ComponentPose=@n93, Alpha=1, ExternalForce="0,0,0") {
    angularDampingOverride=1
    angularSpringConstant=20000
    bAngularSpring=true
    bOverrideAngularDamping=true
    bUseGravityOverride=true
    gravityScale=0.35
    simulationSpace=RootRelative
    alphaBoolBlend {blendOption=Linear}
    boundBone {boneName=dyn_collar_l_01}
    componentLinearAccScale=<0.7 0.7 0.7>
    componentLinearVelScale=<0.7 0.7 0.7>
    gravityOverride {z=-980}
    physicsBodyDefinitions [{
      sphereCollisionRadius=0
      boundBone {boneName=dyn_collar_l_01}
      boxExtents=<20 30 30>
      constraintSetup {angularLimitsMax=<0 5 10>, angularLimitsMin=<0 -2 -10>, angularTarget=<1 0 0>}
      localJointOffset=<4 0 0>
     }]
   }
   n95 = Anim Dynamics - Bone: dyn_collar_l_02(ComponentPose=@n94, Alpha=1, ExternalForce="0,0,0") {
    angularDampingOverride=1
    angularSpringConstant=20000
    bAngularSpring=true
    bOverrideAngularDamping=true
    bUseGravityOverride=true
    gravityScale=0.35
    simulationSpace=RootRelative
    alphaBoolBlend {blendOption=Linear}
    boundBone {boneName=dyn_collar_l_02}
    componentLinearAccScale=<0.7 0.7 0.7>
    componentLinearVelScale=<0.7 0.7 0.7>
    gravityOverride {z=-980}
    physicsBodyDefinitions [{
      sphereCollisionRadius=0
      boundBone {boneName=dyn_collar_l_02}
      boxExtents=<20 30 30>
      constraintSetup {angularLimitsMax=<0 5 10>, angularLimitsMin=<0 -2 -10>, angularTarget=<1 0 0>}
      localJointOffset=<4 0 0>
     }]
   }
   n96 = Anim Dynamics - Bone: dyn_jacket_l_01!(ComponentPose=@n95, Alpha=1, GravityOverride:Vector="0,0,0", ExternalForce="0,0,0") -> (Pose) {
    angularDampingOverride=0.9999
    angularSpringConstant=500
    bAngularSpring=true
    bGravityOverrideInSimSpace=true
    bOverrideAngularDamping=true
    bUseGravityOverride=true
    bUseSphericalLimits=true
    simulationSpace=BoneRelative
    alphaBoolBlend {blendOption=Linear}
    boundBone {boneName=dyn_jacket_l_01}
    componentAppliedLinearAccClamp=<3200 3200 3200>
    componentLinearAccScale=<1.2 1.2 1.2>
    componentLinearVelScale=<1.2 1.2 1.2>
    physicsBodyDefinitions [{
      collisionType=CustomSphere
      sphereCollisionRadius=1
      boundBone {boneName=dyn_jacket_l_01}
      boxExtents=<15 20 20>
      constraintSetup {angularLimitsMax=<1 5 2>, angularLimitsMin=<-1 -15 -10>, angularTarget=<0 0 0>}
      localJointOffset=<3 0 0>
     }]
    relativeSpaceBone {boneName=root}
    sphericalLimits [
     {limitRadius=10, drivingBone {boneName=spine_03}, sphereLocalOffset=<0 0 -2>}
     {limitRadius=9, drivingBone {boneName=spine_04}, sphereLocalOffset=<0 0 -2>}
    ]
   }
   n97 = Anim Dynamics - Bone: dyn_jacket_l_02!(ComponentPose=@n96, Alpha=1, GravityOverride:Vector="0,0,0", ExternalForce="0,0,0") -> (Pose) {
    angularDampingOverride=0.99
    angularSpringConstant=100
    bAngularSpring=true
    bGravityOverrideInSimSpace=true
    bOverrideAngularDamping=true
    bUseGravityOverride=true
    bUseSphericalLimits=true
    gravityScale=0.35
    simulationSpace=BoneRelative
    alphaBoolBlend {blendOption=Linear}
    boundBone {boneName=dyn_jacket_l_02}
    componentAppliedLinearAccClamp=<3000 3000 3000>
    componentLinearAccScale=<1.2 1.2 0.7>
    componentLinearVelScale=<1.2 1.2 0.7>
    physicsBodyDefinitions [{
      collisionType=CustomSphere
      sphereCollisionRadius=1
      boundBone {boneName=dyn_jacket_l_02}
      boxExtents=<15 20 20>
      constraintSetup {angularLimitsMax=<1 5 2>, angularLimitsMin=<-1 -15 -10>, angularTarget=<0 0 0>}
      localJointOffset=<3 0 0>
     }]
    relativeSpaceBone {boneName=root}
    sphericalLimits [
     {limitRadius=10, drivingBone {boneName=spine_02}, sphereLocalOffset=<0 0 -2>}
     {limitRadius=9, drivingBone {boneName=spine_03}, sphereLocalOffset=<0 0 -2>}
    ]
   }
   n98 = Anim Dynamics - Bone: dyn_jacket_bk_01!(ComponentPose=@n97, Alpha=1, GravityOverride:Vector="0,0,0", ExternalForce="0,0,0") -> (Pose) {
    angularDampingOverride=0.9999
    angularSpringConstant=500
    bAngularSpring=true
    bGravityOverrideInSimSpace=true
    bOverrideAngularDamping=true
    bUseGravityOverride=true
    bUseSphericalLimits=true
    simulationSpace=BoneRelative
    alphaBoolBlend {blendOption=Linear}
    alphaScaleBiasClamp {bClampResult=true, clampMax=0.4}
    boundBone {boneName=dyn_jacket_bk_01}
    componentAppliedLinearAccClamp=<1500 1500 1500>
    componentLinearAccScale=<0.6 0.4 0.4>
    componentLinearVelScale=<0.6 0.4 0.4>
    physicsBodyDefinitions [{
      collisionType=CustomSphere
      sphereCollisionRadius=1
      boundBone {boneName=dyn_jacket_bk_01}
      boxExtents=<15 20 20>
      constraintSetup {angularLimitsMax=<1 15 3>, angularLimitsMin=<-1 -15 -20>, angularTarget=<0 0 0>}
      localJointOffset=<3 0 0>
     }]
    relativeSpaceBone {boneName=root}
    sphericalLimits [
     {limitRadius=7, drivingBone {boneName=spine_03}, sphereLocalOffset=<0 0 0>}
     {limitRadius=8, drivingBone {boneName=spine_04}, sphereLocalOffset=<0 0 0>}
    ]
   }
   n99 = Anim Dynamics - Bone: dyn_jacket_bk_02!(ComponentPose=@n98, Alpha=1, GravityOverride:Vector="0,0,0", ExternalForce="0,0,0") -> (Pose) {
    angularDampingOverride=0.99
    angularSpringConstant=100
    bAngularSpring=true
    bGravityOverrideInSimSpace=true
    bOverrideAngularDamping=true
    bUseGravityOverride=true
    bUseSphericalLimits=true
    gravityScale=0.35
    simulationSpace=BoneRelative
    alphaBoolBlend {blendOption=Linear}
    alphaScaleBiasClamp {bClampResult=true, clampMax=0.5}
    boundBone {boneName=dyn_jacket_bk_02}
    componentAppliedLinearAccClamp=<3000 3000 3000>
    componentLinearAccScale=<0.6 0.4 0.4>
    componentLinearVelScale=<0.6 0.4 0.4>
    physicsBodyDefinitions [{
      collisionType=CustomSphere
      sphereCollisionRadius=1
      boundBone {boneName=dyn_jacket_bk_02}
      boxExtents=<15 20 20>
      constraintSetup {angularLimitsMax=<1 -10 3>, angularLimitsMin=<-1 -10 -20>, angularTarget=<0 0 0>}
      localJointOffset=<3 0 0>
     }]
    relativeSpaceBone {boneName=root}
    sphericalLimits [
     {limitRadius=8, drivingBone {boneName=spine_02}, sphereLocalOffset=<0 0 0>}
     {limitRadius=8, drivingBone {boneName=spine_03}, sphereLocalOffset=<0 0 0>}
    ]
   }
   n100 = Anim Dynamics - Bone: dyn_jacket_bk_03!(ComponentPose=@n99, Alpha=1, GravityOverride:Vector="0,0,0", ExternalForce="0,0,0") -> (Pose) {
    angularDampingOverride=0.99
    angularSpringConstant=100
    bAngularSpring=true
    bGravityOverrideInSimSpace=true
    bOverrideAngularDamping=true
    bUseGravityOverride=true
    bUseSphericalLimits=true
    gravityScale=0.35
    simulationSpace=BoneRelative
    alphaBoolBlend {blendOption=Linear}
    alphaScaleBiasClamp {bClampResult=true, clampMax=0.7}
    boundBone {boneName=dyn_jacket_bk_03}
    componentAppliedLinearAccClamp=<3000 3000 3000>
    componentLinearAccScale=<0.6 0.4 0.4>
    componentLinearVelScale=<0.6 0.4 0.4>
    physicsBodyDefinitions [{
      collisionType=CustomSphere
      sphereCollisionRadius=1
      boundBone {boneName=dyn_jacket_bk_03}
      boxExtents=<15 20 20>
      constraintSetup {angularLimitsMax=<1 -10 3>, angularLimitsMin=<-1 -10 -20>, angularTarget=<0 0 0>}
      localJointOffset=<2 0 0>
     }]
    relativeSpaceBone {boneName=root}
    sphericalLimits [
     {limitRadius=10, drivingBone {boneName=spine_01}, sphereLocalOffset=<0 0 0>}
     {limitRadius=9, drivingBone {boneName=spine_02}, sphereLocalOffset=<0 0 0>}
     {limitRadius=15, drivingBone {boneName=pelvis}, sphereLocalOffset=<0 0 0>}
     {limitRadius=15, drivingBone {boneName=thigh_l}, sphereLocalOffset=<0 0 0>}
     {limitRadius=15, drivingBone {boneName=thigh_r}, sphereLocalOffset=<0 0 0>}
    ]
   }
   n101 = Anim Dynamics - Bone: dyn_jacket_r_01!(ComponentPose=@n100, Alpha=1, GravityOverride:Vector="0,0,0", ExternalForce="0,0,0") -> (Pose) {
    angularDampingOverride=0.9999
    angularSpringConstant=500
    bAngularSpring=true
    bGravityOverrideInSimSpace=true
    bOverrideAngularDamping=true
    bUseGravityOverride=true
    bUseSphericalLimits=true
    simulationSpace=BoneRelative
    alphaBoolBlend {blendOption=Linear}
    boundBone {boneName=dyn_jacket_r_01}
    componentAppliedLinearAccClamp=<3200 3200 3200>
    componentLinearAccScale=<1.2 1.2 1.2>
    componentLinearVelScale=<1.2 1.2 1.2>
    physicsBodyDefinitions [{
      collisionType=CustomSphere
      sphereCollisionRadius=1
      boundBone {boneName=dyn_jacket_r_01}
      boxExtents=<15 20 20>
      constraintSetup {angularLimitsMax=<1 3 2>, angularLimitsMin=<-1 -15 -10>, angularTarget=<0 0 0>}
      localJointOffset=<3 0 0>
     }]
    relativeSpaceBone {boneName=root}
    sphericalLimits [
     {limitRadius=10, drivingBone {boneName=spine_03}, sphereLocalOffset=<0 0 0>}
     {limitRadius=9, drivingBone {boneName=spine_04}, sphereLocalOffset=<0 0 0>}
    ]
   }
   n102 = Anim Dynamics - Bone: dyn_jacket_r_02!(ComponentPose=@n101, Alpha=1, GravityOverride:Vector="0,0,0", ExternalForce="0,0,0") -> (Pose) {
    angularDampingOverride=0.99
    angularSpringConstant=100
    bAngularSpring=true
    bGravityOverrideInSimSpace=true
    bOverrideAngularDamping=true
    bUseGravityOverride=true
    bUseSphericalLimits=true
    gravityScale=0.35
    simulationSpace=BoneRelative
    alphaBoolBlend {blendOption=Linear}
    boundBone {boneName=dyn_jacket_r_02}
    componentAppliedLinearAccClamp=<3000 3000 3000>
    componentLinearAccScale=<1.2 1.2 0.7>
    componentLinearVelScale=<1.2 1.2 0.7>
    physicsBodyDefinitions [{
      collisionType=CustomSphere
      sphereCollisionRadius=1
      boundBone {boneName=dyn_jacket_r_02}
      boxExtents=<15 20 20>
      constraintSetup {angularLimitsMax=<1 3 2>, angularLimitsMin=<-1 -15 -10>, angularTarget=<0 0 0>}
      localJointOffset=<3 0 0>
     }]
    relativeSpaceBone {boneName=root}
    sphericalLimits [
     {limitRadius=10, drivingBone {boneName=spine_02}, sphereLocalOffset=<0 0 0>}
     {limitRadius=9, drivingBone {boneName=spine_03}, sphereLocalOffset=<0 0 0>}
    ]
   }
   n103 = Get Gravity Override()
   n104 = Anim Dynamics - Bone: dyn_skirt_bk_01!(ComponentPose=@n102, Alpha=1, GravityOverride:Vector=@n103, ExternalForce="0,0,0") -> (Pose) {
    angularBiasOverride=0.999
    angularDampingOverride=0.999
    angularSpringConstant=1500
    bOverrideAngularBias=true
    bOverrideAngularDamping=true
    bUseGravityOverride=true
    bUsePlanarLimit=false
    bUseSphericalLimits=true
    simulationSpace=RootRelative
    alphaBoolBlend {blendOption=Linear}
    alphaScaleBiasClamp {bClampResult=true, clampMax=0.35}
    boundBone {boneName=dyn_skirt_bk_01}
    componentAppliedLinearAccClamp=<3000 3000 3000>
    componentLinearAccScale=<0.4 0.5 0.5>
    componentLinearVelScale=<0.4 0.5 0.5>
    physicsBodyDefinitions [{
      collisionType=CustomSphere
      sphereCollisionRadius=2
      boundBone {boneName=dyn_skirt_bk_01}
      boxExtents=<8 17 17>
      constraintSetup {angularLimitsMax=<2.5 10 10>, angularLimitsMin=<-2.5 -10 -60>, angularTarget=<0 0 0>}
      localJointOffset=<3 0 0>
     }]
   }
   n105 = Anim Dynamics - Bone: dyn_skirt_bk_02!(ComponentPose=@n104, Alpha=1, GravityOverride:Vector=@n103, ExternalForce="0,0,0") -> (Pose) {
    angularBiasOverride=0.999
    angularDampingOverride=0.999
    angularSpringConstant=1500
    bOverrideAngularBias=true
    bOverrideAngularDamping=true
    bUseGravityOverride=true
    bUsePlanarLimit=false
    bUseSphericalLimits=true
    simulationSpace=RootRelative
    alphaBoolBlend {blendOption=Linear}
    alphaScaleBiasClamp {bClampResult=true, clampMax=0.35}
    boundBone {boneName=dyn_skirt_bk_02}
    componentAppliedLinearAccClamp=<3000 3000 3000>
    componentLinearAccScale=<0.4 0.5 0.5>
    componentLinearVelScale=<0.4 0.5 0.5>
    physicsBodyDefinitions [{
      collisionType=CustomSphere
      sphereCollisionRadius=4
      boundBone {boneName=dyn_skirt_bk_02}
      boxExtents=<8 17 17>
      constraintSetup {angularLimitsMax=<2.5 10 5>, angularLimitsMin=<-2.5 -10 -60>, angularTarget=<0 0 0>}
      localJointOffset=<3 0 0>
     }]
   }
   n106 = "Constraint - dyn_skirt_fr_01 (PP)"!(ComponentPose=@n105, Alpha=1, ConstraintWeights_0=0.5, ConstraintWeights_1:float=0.5) -> (Pose) {
    alphaBoolBlend {blendOption=Linear}
    boneToModify {boneName=dyn_skirt_fr_01}
    constraintSetup [{targetBone {boneName=thigh_l}}, {targetBone {boneName=thigh_r}}]
    constraintWeights [1,1]
   }
   n107 = "Constraint - dyn_skirt_fr_02 (PP)"!(ComponentPose=@n106, Alpha=1, ConstraintWeights_0=0.5, ConstraintWeights_1:float=0.5) -> (Pose) {
    alphaBoolBlend {blendOption=Linear}
    boneToModify {boneName=dyn_skirt_fr_02}
    constraintSetup [{targetBone {boneName=thigh_l}}, {targetBone {boneName=thigh_r}}]
    constraintWeights [1,1]
   }
   n108 = "Constraint - dyn_skirt_bk_01 (PP)"(ComponentPose=@n107, Alpha=0.2, ConstraintWeights_0=0.5, ConstraintWeights_1=0.5) {
    alphaBoolBlend {blendOption=Linear}
    boneToModify {boneName=dyn_skirt_bk_01}
    constraintSetup [{targetBone {boneName=thigh_l}}, {targetBone {boneName=thigh_r}}]
    constraintWeights [1,1]
   }
   n109 = "Constraint - dyn_skirt_bk_02 (PP)"(ComponentPose=@n108, Alpha=0.2, ConstraintWeights_0=0.5, ConstraintWeights_1=0.5) {
    alphaBoolBlend {blendOption=Linear}
    boneToModify {boneName=dyn_skirt_bk_02}
    constraintSetup [{targetBone {boneName=thigh_l}}, {targetBone {boneName=thigh_r}}]
    constraintWeights [1,1]
   }
   n110 = "Constraint - dyn_sleeve_r_bk_01 (P)"(ComponentPose=@n109, Alpha=0.5, ConstraintWeights_0=1) {
    alphaBoolBlend {blendOption=Linear}
    alphaScaleBiasClamp {bClampResult=true, bInterpResult=true, interpSpeedDecreasing=5, interpSpeedIncreasing=7}
    boneToModify {boneName=dyn_sleeve_r_bk_01}
    constraintSetup [{targetBone {boneName=lowerarm_twist_02_r}}]
    constraintWeights [1]
   }
   n111 = "Constraint - dyn_sleeve_r_bk_02 (P)"(ComponentPose=@n110, Alpha=0.5, ConstraintWeights_0=1) {
    alphaBoolBlend {blendOption=Linear}
    alphaScaleBiasClamp {bClampResult=true, clampMax=0.3, interpSpeedDecreasing=5, interpSpeedIncreasing=7}
    boneToModify {boneName=dyn_sleeve_r_bk_02}
    constraintSetup [{targetBone {boneName=lowerarm_twist_02_r}}]
    constraintWeights [1]
   }
   n112 = "Constraint - dyn_sleeve_r_bk_03 (P)"(ComponentPose=@n111, Alpha=0.5, ConstraintWeights_0=1) {
    alphaBoolBlend {blendOption=Linear}
    alphaScaleBiasClamp {bClampResult=true, clampMax=0.01, interpSpeedDecreasing=5, interpSpeedIncreasing=7}
    boneToModify {boneName=dyn_sleeve_r_bk_03}
    constraintSetup [{targetBone {boneName=lowerarm_twist_02_r}}]
    constraintWeights [1]
   }
   n113 = "Constraint - dyn_sleeve_r_bot_01 (P)"(ComponentPose=@n112, Alpha=0.5, ConstraintWeights_0=1) {
    alphaBoolBlend {blendOption=Linear}
    alphaScaleBiasClamp {bClampResult=true, bInterpResult=true, interpSpeedDecreasing=5, interpSpeedIncreasing=7}
    boneToModify {boneName=dyn_sleeve_r_bot_01}
    constraintSetup [{targetBone {boneName=lowerarm_r}}]
    constraintWeights [1]
   }
   n114 = "Constraint - dyn_sleeve_r_bot_02 (P)"(ComponentPose=@n113, Alpha=0.5, ConstraintWeights_0=1) {
    alphaBoolBlend {blendOption=Linear}
    alphaScaleBiasClamp {bClampResult=true, clampMax=0.01, interpSpeedDecreasing=5, interpSpeedIncreasing=7}
    boneToModify {boneName=dyn_sleeve_r_bot_02}
    constraintSetup [{targetBone {boneName=lowerarm_r}}]
    constraintWeights [1]
   }
   n115 = "Constraint - dyn_sleeve_r_bot_03 (P)"(ComponentPose=@n114, Alpha=0.5, ConstraintWeights_0=1) {
    alphaBoolBlend {blendOption=Linear}
    alphaScaleBiasClamp {bClampResult=true, clampMax=0.01, interpSpeedDecreasing=5, interpSpeedIncreasing=7}
    boneToModify {boneName=dyn_sleeve_r_bot_03}
    constraintSetup [{targetBone {boneName=lowerarm_r}}]
    constraintWeights [1]
   }
   n116 = "Constraint - dyn_sleeve_l_bk_01 (P)"(ComponentPose=@n115, Alpha=0.5, ConstraintWeights_0=1) {
    alphaBoolBlend {blendOption=Linear}
    alphaScaleBiasClamp {bClampResult=true, bInterpResult=true, interpSpeedDecreasing=5, interpSpeedIncreasing=7}
    boneToModify {boneName=dyn_sleeve_l_bk_01}
    constraintSetup [{targetBone {boneName=lowerarm_twist_02_l}}]
    constraintWeights [1]
   }
   n117 = "Constraint - dyn_sleeve_l_bk_02 (P)"(ComponentPose=@n116, Alpha=0.5, ConstraintWeights_0=1) {
    alphaBoolBlend {blendOption=Linear}
    alphaScaleBiasClamp {bClampResult=true, clampMax=0.3, interpSpeedDecreasing=5, interpSpeedIncreasing=7}
    boneToModify {boneName=dyn_sleeve_l_bk_02}
    constraintSetup [{targetBone {boneName=lowerarm_twist_02_l}}]
    constraintWeights [1]
   }
   n118 = "Constraint - dyn_sleeve_l_bk_03 (P)"(ComponentPose=@n117, Alpha=0.5, ConstraintWeights_0=1) {
    alphaBoolBlend {blendOption=Linear}
    alphaScaleBiasClamp {bClampResult=true, clampMax=0.01, interpSpeedDecreasing=5, interpSpeedIncreasing=7}
    boneToModify {boneName=dyn_sleeve_l_bk_03}
    constraintSetup [{targetBone {boneName=lowerarm_twist_02_l}}]
    constraintWeights [1]
   }
   n119 = "Constraint - dyn_sleeve_l_bot_01 (P)"(ComponentPose=@n118, Alpha=0.5, ConstraintWeights_0=1) {
    alphaBoolBlend {blendOption=Linear}
    alphaScaleBiasClamp {bClampResult=true, bInterpResult=true, interpSpeedDecreasing=5, interpSpeedIncreasing=7}
    boneToModify {boneName=dyn_sleeve_l_bot_01}
    constraintSetup [{targetBone {boneName=lowerarm_l}}]
    constraintWeights [1]
   }
   n120 = "Constraint - dyn_sleeve_l_bot_02 (P)"(ComponentPose=@n119, Alpha=0.5, ConstraintWeights_0=1) {
    alphaBoolBlend {blendOption=Linear}
    alphaScaleBiasClamp {bClampResult=true, clampMax=0.01, interpSpeedDecreasing=5, interpSpeedIncreasing=7}
    boneToModify {boneName=dyn_sleeve_l_bot_02}
    constraintSetup [{targetBone {boneName=lowerarm_l}}]
    constraintWeights [1]
   }
   n121 = "Constraint - dyn_sleeve_l_bot_03 (P)"(ComponentPose=@n120, Alpha=0.5, ConstraintWeights_0=1) {
    alphaBoolBlend {blendOption=Linear}
    alphaScaleBiasClamp {bClampResult=true, clampMax=0.01, interpSpeedDecreasing=5, interpSpeedIncreasing=7}
    boneToModify {boneName=dyn_sleeve_l_bot_03}
    constraintSetup [{targetBone {boneName=lowerarm_l}}]
    constraintWeights [1]
   }
   n122 = Control Rig!(
    Source=Component To Local(ComponentPose=@n121)
    Alpha
    Crouch:double=Get Crouch() -> (Crouch:double)
    Tactical:double=Get Tactical_Sprint() -> (Tactical_Sprint:double)
    Move:double=Get Moving() -> (Moving:double)
    Emote:double=Get Emote() -> (Emote:double)
    RigidBody:bool=Get bRigidBodyEnabled()
    Melee:double=Get Melee() -> (Melee:double)
    Skydive:double=Get Skydive() -> (Skydive:double)
    Left:double=Get Left() -> (Left:double)
    Right:double=Get Right() -> (Right:double)
    ADS:double=Get ADS() -> (ADS:double)
   ) -> (Pose) {
    alphaBoolBlend {blendOption=Linear}
    controlRigAssetReference {blueprintRigClass=&/F_MED_BlazerVeil_CtrlRig.F_MED_BlazerVeil_CtrlRig_C}
   }
   Save cached pose 'StartupCache'(Pose=@n122) {cacheName=StartupCache}
  ]}
 EventGraph {nodes [
   n1 = Cloth Controller Function DEPRECATED!(
    execute
    CharacterPartAnimInstance:Custom Character Part Anim Instance Object Reference=Self-Reference()
    ClothingAssetName:String=Sleeve_02_0
    Cloth Paramaters:Cloth Params Struct Structure=Get Cloth Paramaters_0() -> (Cloth Paramaters_0:Cloth Params Struct Structure)
    Cloth Core Settings:Cloth Core Setting Struct Structure=Get Cloth Core Settings_0() -> (Cloth Core Settings_0:Cloth Core Setting Struct Structure)
    ClothUpAxis:Vector="1,0,0"
    OnGround_Moving_SpeedLimit:double=2
    OnGround_Moving_AngularVel_Max:double=0.05
    OnGround_Moving_FictitiousVel_Max:double=0.6
    LinVelScale_FwdBwd_Clamp:double=0
    LinVelScale_UpDwn_Clamp:double=0.7
    DeltaTime:double=Get DeltaTimeLocal() -> (DeltaTimeLocal:double)
    GameTimeInSeconds:double=GetGameTimeInSeconds()
    WindGustsParams:Wind Gust Params Struct Structure=Get Wind Gust Params() -> (Wind Gust Params:Wind Gust Params Struct Structure)
    WindGustLocalSettings:Wind Gust Local Setting Struct Structure=Get Gust Local Settings() -> (Gust Local Settings:Wind Gust Local Setting Struct Structure)
   ) -> (
    then
    CurrentStateOut:String
    Cloth Settings Out:Cloth Setting Struct Structure
    Cloth Core Settings Out:Cloth Core Setting Struct Structure
    WindGustSettings Out:Wind Gust Setting Struct Structure
    WindGustLocalSettings Out:Wind Gust Local Setting Struct Structure
    GustStrength - WindVelocityOut:Vector
    LinearVelVector-DELETE LATER:Vector
   )
   > Set Gust Local Settings(Gust Local Settings:Wind Gust Local Setting Struct Structure=@n1.WindGustLocalSettings Out) -> (Output_Get:Wind Gust Local Setting Struct Structure)
   Event BlueprintUpdateAnimation()
   > n2 = Sequence!(execute) -> (then_0, then_1, then_2:Exec, then_3:Exec, then_4:Exec, then_5:Exec, then_6:Exec)
   > n3 = Gravity Override State Function DEPRECATED(CharacterPartAnimInstance=Self-Reference(), GameTimeInSeconds=GetGameTimeInSeconds(), GravityOverrideParams=Get Gravity Override Params())
   > Set Gravity Override(Gravity Override=@n3)
   Set Tactical_Sprint(execute=@n2.then_1, Tactical_Sprint:double=SelectFloat(A=1, B=0, bPickA=Get bIsTacticalSprinting())) -> (Output_Get:double)
   n4 = Get bIsMoving2D()
   > Set Moving(Moving:double=SelectFloat(A=1, B=0, bPickA=@n4)) -> (Output_Get:double)
   n5 = Get bIsCrouching()
   > Set Crouch(Crouch:double=SelectFloat(A=1, B=0, bPickA=@n5)) -> (Output_Get:double)
   > Set Crouch_off(Crouch_off:double=SelectFloat(A=0, B=1, bPickA=AND Boolean(A=@n5, B=@n4))) -> (Output_Get:double)
   > Set Emote(Emote:double=SelectFloat(A=1, B=0, bPickA=Get bIsPlayingEmote())) -> (Output_Get:double)
   > Set Melee(Melee:double=SelectFloat(A=1, B=0, bPickA=Get bIsPlayingMeleeAnim())) -> (Output_Get:double)
   n6 = CheckForBackbling()
   > Set WearingBackbling(WearingBackbling:bool=OR Boolean!(A=@n6, B=@n6.bIsBackblingACape, C:bool=@n6.bIsBackblingLong) -> (ReturnValue)) -> (Output_Get:bool)
   > Set Skydive(Skydive:double=SelectFloat(A=1, B=0, bPickA=Get bIsSkyDiving())) -> (Output_Get:double)
   > Set ADS(ADS:double=SelectFloat(A=1, B=0, bPickA=Get bIsTargeting())) -> (Output_Get:double)
   n7 = Branch(execute=@n2.then_2, Condition=Get bIsSliding())
   > n8 = Set Left_Calf(Left_Calf:double=GetAlphaFromReferenceTransformDelta(SkeletalMeshComponent=GetPartSkeletalMeshComponent(), BoneName=calf_l, BaseName=thigh_l, InputRange=(X=0,Y=1), TwistAxis="1,0,0").RawOutput) -> (Output_Get:double)
   > n9 = Branch(Condition=NOT Boolean(A="float > float"(A=@n8, B=100)))
   > Set Base_Slide(Base_Slide:double=1) -> (Output_Get:double)
   > Set Knee_Slide(Knee_Slide:double=0) -> (Output_Get:double)
   Set Base_Slide(execute=@n9.else, Base_Slide=0)
   > Set Knee_Slide(Knee_Slide=1)
   Set Base_Slide(execute=@n7.else, Base_Slide=0)
   > Set Knee_Slide(Knee_Slide=0)
   n10 = PlayerMovement(execute=@n2.then_4, AnimInstance=Self-Reference(), XDotProductCoefficient=100, YDotProductCoefficient=100, ZDotProductCoefficient=100)
   n11 = BreakVector(InVec=MakeVector(X=@n10, Y=@n10.YMovement))
   > Set Left(Left:double="To Float (Boolean)"(InBool="float < float"(A=@n11, B=-50))) -> (Output_Get:double)
   > Set Right(Right:double="To Float (Boolean)"(InBool="float > float"(A=@n11, B=50))) -> (Output_Get:double)
   > Set Forward(Forward:double="To Float (Boolean)"(InBool="float > float"(A=@n11.Y, B=40))) -> (Output_Get:double)
   > Set Backward(Backward:double="To Float (Boolean)"(InBool="float < float"(A=@n11.Y, B=-40))) -> (Output_Get:double)
   n12 = GetOwnerSkeletalMeshComponent()
   n13 = BreakRotator(InRot=BreakTransform(InTransform=GetDeltaTransformFromRefPose(self=@n12, BoneName=lowerarm_r, BaseName=spine_05)).Rotation)
   n14 = OR Boolean!(A=Get bIsTargeting(), B=Get bIsPlayingEmote()) -> (ReturnValue)
   Set arm_up_r(
    execute=@n2.then_5
    arm_up_r:double=SelectFloat(A=MapRangeClamped(Value=@n13.Yaw, InRangeA=45, InRangeB=20, OutRangeB=0.8), B=0, bPickA=NOT Boolean(A=@n14))
   ) -> (Output_Get:double)
   n15 = GetOwnerSkeletalMeshComponent()
   n16 = BreakRotator(InRot=BreakTransform(InTransform=GetDeltaTransformFromRefPose(self=@n15, BoneName=lowerarm_l, BaseName=spine_05)).Rotation)
   n17 = OR Boolean(A=Get bIsTargeting(), B=Get bIsPlayingEmote())
   > Set arm_up_l(arm_up_l:double=SelectFloat(A=MapRangeClamped(Value=@n16.Yaw, InRangeA=45, InRangeB=20, OutRangeB=0.5), B=0, bPickA=NOT Boolean(A=@n17))) -> (Output_Get:double)
   Branch(execute=@n2.then_6, Condition=Get bIsInFrontEnd())
   > Set Custom Idle Is Playing(Custom Idle Is Playing:float=GetCurveValue(self=GetOwnerFrontEndAnimInstance(), CurveName=Frontend_Idle_Bypass)) -> (Output_Get:float)
   n18 = GetClothingSimulationInteractor!(self:Skeletal Mesh Component Object Reference=GetOwningComponent(), ClothingAssetName:Name=None) -> (ReturnValue:Clothing Simulation Interactor Object Reference)
   Is Valid(exec=@n2.then_3, InputObject=@n18)
   n19 = MakeLiteralString(Value=Silencer_Cloth)
   > n20 = Cast To ChaosClothingInteractor!(
    execute
    Object=GetClothingInteractor!(self:Clothing Simulation Interactor Object Reference=@n18, ClothingAssetName:Name=String To Name(InString=@n19)) -> (ReturnValue:Clothing Interactor Object Reference)
   ) -> (then, CastFailed, AsChaos Clothing Interactor:Chaos Clothing Interactor Object Reference) {is=title K2Node_DynamicCast}
   > Set ChaosClothingInteractor(ChaosClothingInteractor:Chaos Clothing Interactor Object Reference=@n20) -> (Output_Get:Chaos Clothing Interactor Object Reference)
   Event OnBeginPlayingEmote() {is=title K2Node_Event}
   > Branch(Condition=Get bIsInFrontEnd())
   > Delay(Duration=0.08)
   > ResetDynamics(InTeleportType=TeleportPhysics)
   > ResetDynamics(InTeleportType=ResetPhysics)
   Event OnCharacterCustomizationLoaded()
   > n21 = Set SkelMesh(SkelMesh=GetPartSkeletalMeshComponent())
   > n22 = Get Dynamic Material Instance(SkeletalMeshComponent=@n21)
   > Set Mat Instance Dynamic(Mat Instance Dynamic=@n22)
   > n23 = GetOrCreateDynamicOverlayMaterial(MeshComponent=@n21, MaterialName=None)
   > Set Overlay Mat Instance Dynamic(Overlay Mat Instance Dynamic=@n23)
   Event OnEndFalling()
   > Branch(Condition=Get resetClothLandingFromZilpine() -> (resetClothLandingFromZilpine:bool))
   n24 = Get ChaosClothingInteractor() -> (ChaosClothingInteractor:Chaos Clothing Interactor Object Reference)
   > ResetAndTeleport!(execute, self:Chaos Clothing Interactor Object Reference=@n24, bReset:bool=true, bTeleport:bool=true) -> (then)
   > Set resetClothLandingFromZilpine(resetClothLandingFromZilpine:bool=false) -> (Output_Get:bool)
   n25 = Event BlueprintProcessAnimInputEvent()
   > n26 = Branch(Condition=IsNamed(self=@n25.AnimInputEvent, InEventName=aie.Player.Zipline.Entered))
   > Delay(Duration=0.1)
   > ResetAndTeleport(self=@n24, bReset=true, bTeleport=true)
   Branch(execute=@n26.else, Condition=IsNamed(self=@n25.AnimInputEvent, InEventName=aie.Player.Zipline.Exited))
   > Set resetClothLandingFromZilpine(resetClothLandingFromZilpine=true)
   Event BlueprintInitializeAnimation()
   > n27 = Sequence!(execute) -> (then_0, then_1)
   > Set SkelMesh(SkelMesh=GetOwnerSkeletalMeshComponent())
   n28 = TryGetPawnOwner()
   > Is Valid(InputObject=@n28)
   > n29 = Cast To FortPlayerPawnAthena(Object=@n28)
   > Set PlayerPawn(PlayerPawn:Fort Player Pawn Athena Object Reference=@n29) -> (Output_Get:Fort Player Pawn Athena Object Reference)
   n30 = GetPartSkeletalMeshComponent()
   Set bResetAfterTeleport!(execute=@n27.then_1, bResetAfterTeleport:bool=false, self:Skeletal Mesh Component Object Reference=@n30) -> (then, Output_Get:bool)
   > Set TeleportDistanceThreshold!(execute, TeleportDistanceThreshold:float=300, self:Skeletal Mesh Component Object Reference=@n30) -> (then, Output_Get:float)
   > Set TeleportRotationThreshold!(execute, TeleportRotationThreshold:float=15, self:Skeletal Mesh Component Object Reference=@n30) -> (then, Output_Get:float)
   Sequence()
   Branch(Condition=Get bClothEnabled())
   Reroute Node(InputPin:Wildcard) -> (OutputPin:Wildcard)
  ]}
]
variables [
 ADS {type=number}
 arm_up_l {type=number}
 arm_up_r {type=number}
 Backward {type=number}
 Base_Slide {type=number}
 ChaosClothingInteractor {type=ChaosClothingInteractor, default=null}
 Cloth Core Settings_0 {
  type=ClothCoreSettingStruct
  default {
   collisionThickness=1.5
   dampingCoefficient=0.1
   frictionCoefficient=0.5
   localDampingCoefficient=0.1
   numberIterations=2
   numberSubsteps=1
   selfCollisionThickness=0
   animDriveDamping=<0 1>
   animDriveStiffness=<0 1>
   areaStiffness=<1 1>
   bendingStiffness=<1 1>
   edgeStiffness=<1 1>
   tetherStiffness=<1 1>
  }
 }
 Cloth Paramaters_0 {
  type=ClothParamsStruct
  default {
   dBNO {
    angularVelocityScale=0.6
    fictitiousAngularScale=0.1
    jointName=pelvis
    dragCoefficient=<0.2 1>
    gravityOverride=<-980 400 0>
    liftCoefficient=<0 0.5>
    linearVelocityScale=<0.3 0.3 0.1>
   }
   dBNO_Carried {
    angularVelocityScale=0.6
    fictitiousAngularScale=0.1
    jointName=pelvis
    dragCoefficient=<0.2 1>
    gravityOverride=<-980 -200 0>
    liftCoefficient=<0 0.5>
    linearVelocityScale=<0.3 0.3 0.1>
   }
   dBNO_Impulse {
    angularVelocityScale=0
    fictitiousAngularScale=0
    jointName=pelvis
    dragCoefficient=<0 0.2>
    gravityOverride=<-3000 -2000 0>
    liftCoefficient=<0 0.1>
    linearVelocityScale=<0.3 0.1 0.3>
   }
   dBNO_RollBwd {
    angularVelocityScale=0
    fictitiousAngularScale=0
    jointName=pelvis
    dragCoefficient=<0 0.2>
    gravityOverride=<-2000 0 0>
    liftCoefficient=<0 0.1>
    linearVelocityScale=<0.3 0.1 0.3>
   }
   dBNO_RollFwd {
    angularVelocityScale=0
    fictitiousAngularScale=0
    jointName=pelvis
    dragCoefficient=<0 0.2>
    gravityOverride=<-980 0 0>
    liftCoefficient=<0 0.1>
    linearVelocityScale=<0.3 0.1 0.3>
   }
   dBNO_Tumbling {
    angularVelocityScale=0
    fictitiousAngularScale=0
    jointName=pelvis
    dragCoefficient=<0 0.2>
    gravityOverride=<-2500 -300 0>
    liftCoefficient=<0 0.1>
    linearVelocityScale=<0.3 0.1 0.3>
   }
   emoteOrMelee {
    angularVelocityScale=0.2
    fictitiousAngularScale=0.1
    jointName=pelvis
    dragCoefficient=<0 0.3>
    gravityOverride=<-980 -500 0>
    liftCoefficient=<0 0.2>
    linearVelocityScale=<0.05 0.05 0.05>
   }
   front_End {
    angularVelocityScale=0.1
    fictitiousAngularScale=0.2
    jointName=root
    dragCoefficient=<0.2 0.2>
    gravityOverride=<0 -200 -980>
    liftCoefficient=<0.1 0.1>
    linearVelocityScale=<0.1 0.1 0.1>
   }
   onGround_Crouch_Moving {
    angularVelocityScale=0.1
    fictitiousAngularScale=0.3
    jointName=root
    dragCoefficient=<1 1>
    gravityOverride=<0 0 -800>
    liftCoefficient=<0.5 0.5>
    linearVelocityScale=<0.4 0.4 0.3>
   }
   onGround_Falling {
    angularVelocityScale=0.6
    fictitiousAngularScale=0
    jointName=root
    dragCoefficient=<0 0.2>
    gravityOverride=<0 -500 -980>
    liftCoefficient=<0 0.1>
    linearVelocityScale=<2 1 0.3>
   }
   onGround_Moving {
    angularVelocityScale=0.6
    fictitiousAngularScale=0.1
    jointName=root
    dragCoefficient=<0 0.3>
    gravityOverride=<0 -100 -980>
    liftCoefficient=<0 0.2>
    linearVelocityScale=<0.5 1 1>
   }
   onGround_Standing {
    angularVelocityScale=0.05
    fictitiousAngularScale=0.6
    jointName=root
    dragCoefficient=<0 2>
    gravityOverride=<0 -100 -980>
    liftCoefficient=<0 0.5>
    linearVelocityScale=<0.45 1 1>
   }
   rail_Grinding {
    angularVelocityScale=0.2
    fictitiousAngularScale=0
    jointName=root
    dragCoefficient=<0.1 0.2>
    gravityOverride=<10 -300 -200>
    liftCoefficient=<0 0.5>
    linearVelocityScale=<0.1 0.1 0.1>
   }
   riding_Moving {
    angularVelocityScale=0.05
    fictitiousAngularScale=1
    jointName=root
    dragCoefficient=<0.5 0.5>
    gravityOverride=<0 0 -980>
    liftCoefficient=<0.5 0.5>
    linearVelocityScale=<0.4 0.6 0.3>
   }
   riding_Moving_Targeting {
    angularVelocityScale=0.05
    fictitiousAngularScale=0.7
    jointName=root
    dragCoefficient=<0.5 0.5>
    gravityOverride=<0 0 -980>
    liftCoefficient=<0.5 0.5>
    linearVelocityScale=<0.2 0.6 0.2>
   }
   riding_Standing {
    angularVelocityScale=0.05
    fictitiousAngularScale=1
    jointName=root
    dragCoefficient=<0.5 0.5>
    gravityOverride=<0 -450 -980>
    liftCoefficient=<0.5 0.5>
    linearVelocityScale=<0.6 0.6 0.3>
   }
   skydive_DivingDown {
    angularVelocityScale=0.1
    fictitiousAngularScale=0
    jointName=pelvis
    dragCoefficient=<0 1.5>
    gravityOverride=<-450 -100 0>
    liftCoefficient=<0 0.3>
    linearVelocityScale=<0.1 0.1 0.1>
   }
   skydive_FloatingDown {
    angularVelocityScale=0.1
    fictitiousAngularScale=0
    jointName=root
    dragCoefficient=<0 2.5>
    gravityOverride=<0 0 700>
    liftCoefficient=<0 0.4>
    linearVelocityScale=<0.2 0.2 0.1>
   }
   skydive_GoingUp {
    angularVelocityScale=0.1
    fictitiousAngularScale=0
    jointName=root
    dragCoefficient=<0 0>
    gravityOverride=<0 -3000 3000>
    liftCoefficient=<0 0>
    linearVelocityScale=<0.1 0.1 0.05>
   }
   skydive_Parachuting {
    angularVelocityScale=0.1
    fictitiousAngularScale=0
    jointName=root
    dragCoefficient=<0 2.5>
    gravityOverride=<0 0 -980>
    liftCoefficient=<0 0.5>
    linearVelocityScale=<0.2 0.2 0.2>
   }
   skydive_Surfing {
    angularVelocityScale=0.6
    fictitiousAngularScale=0.1
    jointName=root
    dragCoefficient=<1.5 1.5>
    gravityOverride=<0 -600 -500>
    liftCoefficient=<1 1>
    linearVelocityScale=<0.1 0.1 0.3>
   }
   swimming {
    angularVelocityScale=0.6
    fictitiousAngularScale=0
    jointName=root
    dragCoefficient=<2.5 2.5>
    gravityOverride=<0 0 0>
    liftCoefficient=<0.1 0.1>
    linearVelocityScale=<0.6 0.6 0.5>
   }
   tactical_Sprint {
    angularVelocityScale=0.6
    fictitiousAngularScale=0.1
    jointName=root
    dragCoefficient=<0.3 0.3>
    gravityOverride=<0 -100 -980>
    liftCoefficient=<0.2 0.2>
    linearVelocityScale=<0.2 0.2 0.2>
   }
   ziplining {
    angularVelocityScale=0.6
    fictitiousAngularScale=0
    jointName=root
    dragCoefficient=<0 1.5>
    gravityOverride=<0 0 -980>
    liftCoefficient=<0 0.3>
    linearVelocityScale=<0.4 0.4 0.5>
   }
  }
 }
 Crouch {type=number}
 Crouch_off {type=number}
 Custom Idle Is Playing {type=number}
 DeltaTimeLocal {type=number}
 Emote {type=number}
 Forward {type=number}
 Gravity Override {type=Vector}
 Gravity Override Params {type=GravityOverrideParamsStruct}
 Gust Local Settings {type=WindGustLocalSettingStruct, default {elapsedTime=0, finalGustStrengthLocal=0, gustDuration=0, gustStrength=0, timeBetweenGusts=0}}
 Knee_Slide {type=number}
 Left {type=number}
 Left_Calf {type=number}
 Local Noise Offset {type=Vector}
 Local Noise Transform Location {type=Vector}
 Mat Instance Dynamic {type=MaterialInstanceDynamic}
 Melee {type=number}
 Moving {type=number}
 Overlay Mat Instance Dynamic {type=MaterialInstanceDynamic}
 PlayerPawn {type=FortPlayerPawnAthena}
 R1 Velocity Transform Location {type=Vector}
 R1 Velocity Transform Velocity {type=Vector}
 R2 Velocity Transform Location {type=Vector}
 R2 Velocity Transform Velocity {type=Vector}
 resetClothLandingFromZilpine {type=boolean}
 Right {type=number}
 SkelMesh {type=SkeletalMeshComponent}
 Skydive {type=number}
 Tactical_Sprint {type=number}
 Velocity Interp Speed {type=number}
 WearingBackbling {type=boolean}
 Wind Gust Params {
  type=WindGustParamsStruct
  default {
   dBNO {
    maxDownInterpSpeed=5
    maxGustDuration=2
    maxGustStrength=7
    maxTimeBetweenGusts=7
    maxUpInterpSpeed=5
    minDownInterpSpeed=2
    minGustDuration=0.5
    minGustStrength=3
    minTimeBetweenGusts=0
    minUpInterpSpeed=2
    windTurbulenceScalarA=0.1
    windTurbulenceScalarB=1
    windAmplitude=<0 0 0>
    windFrequency=<0 0 0>
    windOffset=<0 0 0>
   }
   dBNO_Carried {
    maxDownInterpSpeed=5
    maxGustDuration=2
    maxGustStrength=7
    maxTimeBetweenGusts=7
    maxUpInterpSpeed=5
    minDownInterpSpeed=2
    minGustDuration=0.5
    minGustStrength=3
    minTimeBetweenGusts=0
    minUpInterpSpeed=2
    windTurbulenceScalarA=0.1
    windTurbulenceScalarB=1
    windAmplitude=<0 0 0>
    windFrequency=<0 0 0>
    windOffset=<0 0 0>
   }
   dBNO_Impulse {
    maxDownInterpSpeed=5
    maxGustDuration=2
    maxGustStrength=7
    maxTimeBetweenGusts=7
    maxUpInterpSpeed=5
    minDownInterpSpeed=2
    minGustDuration=0.5
    minGustStrength=3
    minTimeBetweenGusts=0
    minUpInterpSpeed=2
    windTurbulenceScalarA=0.1
    windTurbulenceScalarB=1
    windAmplitude=<0 0 0>
    windFrequency=<0 0 0>
    windOffset=<0 0 0>
   }
   dBNO_RollBwd {
    maxDownInterpSpeed=5
    maxGustDuration=2
    maxGustStrength=7
    maxTimeBetweenGusts=7
    maxUpInterpSpeed=5
    minDownInterpSpeed=2
    minGustDuration=0.5
    minGustStrength=3
    minTimeBetweenGusts=0
    minUpInterpSpeed=2
    windTurbulenceScalarA=0.1
    windTurbulenceScalarB=1
    windAmplitude=<0 0 0>
    windFrequency=<0 0 0>
    windOffset=<0 0 0>
   }
   dBNO_RollFwd {
    maxDownInterpSpeed=5
    maxGustDuration=2
    maxGustStrength=7
    maxTimeBetweenGusts=7
    maxUpInterpSpeed=5
    minDownInterpSpeed=2
    minGustDuration=0.5
    minGustStrength=3
    minTimeBetweenGusts=0
    minUpInterpSpeed=2
    windTurbulenceScalarA=0.1
    windTurbulenceScalarB=1
    windAmplitude=<0 0 0>
    windFrequency=<0 0 0>
    windOffset=<0 0 0>
   }
   dBNO_Tumbling {
    maxDownInterpSpeed=5
    maxGustDuration=2
    maxGustStrength=7
    maxTimeBetweenGusts=7
    maxUpInterpSpeed=5
    minDownInterpSpeed=2
    minGustDuration=0.5
    minGustStrength=3
    minTimeBetweenGusts=0
    minUpInterpSpeed=2
    windTurbulenceScalarA=0.1
    windTurbulenceScalarB=1
    windAmplitude=<0 0 0>
    windFrequency=<0 0 0>
    windOffset=<0 0 0>
   }
   emoteOrMelee {
    maxDownInterpSpeed=5
    maxGustDuration=2
    maxGustStrength=7
    maxTimeBetweenGusts=7
    maxUpInterpSpeed=5
    minDownInterpSpeed=2
    minGustDuration=0.5
    minGustStrength=3
    minTimeBetweenGusts=0
    minUpInterpSpeed=2
    windTurbulenceScalarA=0.1
    windTurbulenceScalarB=1
    windAmplitude=<0 0 0>
    windFrequency=<0 0 0>
    windOffset=<0 0 0>
   }
   front_End {
    maxDownInterpSpeed=5
    maxGustDuration=2
    maxGustStrength=7
    maxTimeBetweenGusts=7
    maxUpInterpSpeed=5
    minDownInterpSpeed=2
    minGustDuration=0.5
    minGustStrength=3
    minTimeBetweenGusts=0
    minUpInterpSpeed=2
    windTurbulenceScalarA=0.1
    windTurbulenceScalarB=1
    windAmplitude=<0 0 0>
    windFrequency=<0 0 0>
    windOffset=<0 0 0>
   }
   onGround_Crouch_Moving {
    maxDownInterpSpeed=5
    maxGustDuration=2
    maxGustStrength=7
    maxTimeBetweenGusts=7
    maxUpInterpSpeed=5
    minDownInterpSpeed=2
    minGustDuration=0.5
    minGustStrength=3
    minTimeBetweenGusts=0
    minUpInterpSpeed=2
    windTurbulenceScalarA=0.1
    windTurbulenceScalarB=1
    windAmplitude=<0 0 0>
    windFrequency=<0 0 0>
    windOffset=<0 0 0>
   }
   onGround_Falling {
    maxDownInterpSpeed=5
    maxGustDuration=2
    maxGustStrength=7
    maxTimeBetweenGusts=7
    maxUpInterpSpeed=5
    minDownInterpSpeed=2
    minGustDuration=0.5
    minGustStrength=3
    minTimeBetweenGusts=0
    minUpInterpSpeed=2
    windTurbulenceScalarA=0.1
    windTurbulenceScalarB=1
    windAmplitude=<0 0 0>
    windFrequency=<0 0 0>
    windOffset=<0 0 0>
   }
   onGround_Moving {
    maxDownInterpSpeed=5
    maxGustDuration=2
    maxGustStrength=7
    maxTimeBetweenGusts=7
    maxUpInterpSpeed=5
    minDownInterpSpeed=2
    minGustDuration=0.5
    minGustStrength=3
    minTimeBetweenGusts=0
    minUpInterpSpeed=2
    windTurbulenceScalarA=0.1
    windTurbulenceScalarB=1
    windAmplitude=<0 0 0>
    windFrequency=<0 0 0>
    windOffset=<0 0 0>
   }
   onGround_Standing {
    maxDownInterpSpeed=5
    maxGustDuration=2
    maxGustStrength=7
    maxTimeBetweenGusts=7
    maxUpInterpSpeed=5
    minDownInterpSpeed=2
    minGustDuration=0.5
    minGustStrength=3
    minTimeBetweenGusts=0
    minUpInterpSpeed=2
    windTurbulenceScalarA=0.1
    windTurbulenceScalarB=1
    windAmplitude=<0 0 0>
    windFrequency=<0 0 0>
    windOffset=<0 0 0>
   }
   rail_Grinding {
    maxDownInterpSpeed=5
    maxGustDuration=8
    maxGustStrength=7
    maxTimeBetweenGusts=8
    maxUpInterpSpeed=5
    minDownInterpSpeed=2
    minGustDuration=0.2
    minGustStrength=3
    minTimeBetweenGusts=1
    minUpInterpSpeed=2
    windTurbulenceScalarA=0.1
    windTurbulenceScalarB=0.8
    windAmplitude=<20 65 42>
    windFrequency=<5 6 8.6>
    windOffset=<50 -500 -400>
   }
   riding_Moving {
    maxDownInterpSpeed=5
    maxGustDuration=2
    maxGustStrength=7
    maxTimeBetweenGusts=7
    maxUpInterpSpeed=5
    minDownInterpSpeed=2
    minGustDuration=0.5
    minGustStrength=3
    minTimeBetweenGusts=0
    minUpInterpSpeed=2
    windTurbulenceScalarA=0.1
    windTurbulenceScalarB=1
    windAmplitude=<0 0 0>
    windFrequency=<0 0 0>
    windOffset=<0 0 0>
   }
   riding_Moving_Targeting {
    maxDownInterpSpeed=5
    maxGustDuration=2
    maxGustStrength=7
    maxTimeBetweenGusts=7
    maxUpInterpSpeed=5
    minDownInterpSpeed=2
    minGustDuration=0.5
    minGustStrength=3
    minTimeBetweenGusts=0
    minUpInterpSpeed=2
    windTurbulenceScalarA=0.1
    windTurbulenceScalarB=1
    windAmplitude=<0 0 0>
    windFrequency=<0 0 0>
    windOffset=<0 0 0>
   }
   riding_Standing {
    maxDownInterpSpeed=5
    maxGustDuration=2
    maxGustStrength=7
    maxTimeBetweenGusts=7
    maxUpInterpSpeed=5
    minDownInterpSpeed=2
    minGustDuration=0.5
    minGustStrength=3
    minTimeBetweenGusts=0
    minUpInterpSpeed=2
    windTurbulenceScalarA=0.1
    windTurbulenceScalarB=1
    windAmplitude=<0 0 0>
    windFrequency=<0 0 0>
    windOffset=<0 0 0>
   }
   skydive_DivingDown {
    maxDownInterpSpeed=5
    maxGustDuration=2
    maxGustStrength=7
    maxTimeBetweenGusts=7
    maxUpInterpSpeed=5
    minDownInterpSpeed=2
    minGustDuration=0.5
    minGustStrength=3
    minTimeBetweenGusts=0
    minUpInterpSpeed=2
    windTurbulenceScalarA=0.1
    windTurbulenceScalarB=1
    windAmplitude=<0 0 0>
    windFrequency=<0 0 0>
    windOffset=<0 0 0>
   }
   skydive_FloatingDown {
    maxDownInterpSpeed=5
    maxGustDuration=2
    maxGustStrength=7
    maxTimeBetweenGusts=7
    maxUpInterpSpeed=5
    minDownInterpSpeed=2
    minGustDuration=0.5
    minGustStrength=3
    minTimeBetweenGusts=0
    minUpInterpSpeed=2
    windTurbulenceScalarA=0.1
    windTurbulenceScalarB=1
    windAmplitude=<0 0 0>
    windFrequency=<0 0 0>
    windOffset=<0 0 0>
   }
   skydive_GoingUp {
    maxDownInterpSpeed=5
    maxGustDuration=2
    maxGustStrength=7
    maxTimeBetweenGusts=7
    maxUpInterpSpeed=5
    minDownInterpSpeed=2
    minGustDuration=0.5
    minGustStrength=3
    minTimeBetweenGusts=0
    minUpInterpSpeed=2
    windTurbulenceScalarA=0.1
    windTurbulenceScalarB=1
    windAmplitude=<0 0 0>
    windFrequency=<0 0 0>
    windOffset=<0 0 0>
   }
   skydive_Parachuting {
    maxDownInterpSpeed=5
    maxGustDuration=2
    maxGustStrength=7
    maxTimeBetweenGusts=7
    maxUpInterpSpeed=5
    minDownInterpSpeed=2
    minGustDuration=0.5
    minGustStrength=3
    minTimeBetweenGusts=0
    minUpInterpSpeed=2
    windTurbulenceScalarA=0.1
    windTurbulenceScalarB=1
    windAmplitude=<0 0 0>
    windFrequency=<0 0 0>
    windOffset=<0 0 0>
   }
   skydive_Surfing {
    maxDownInterpSpeed=5
    maxGustDuration=2
    maxGustStrength=7
    maxTimeBetweenGusts=7
    maxUpInterpSpeed=5
    minDownInterpSpeed=2
    minGustDuration=0.5
    minGustStrength=3
    minTimeBetweenGusts=0
    minUpInterpSpeed=2
    windTurbulenceScalarA=0.1
    windTurbulenceScalarB=1
    windAmplitude=<0 0 0>
    windFrequency=<0 0 0>
    windOffset=<0 0 0>
   }
   swimming {
    maxDownInterpSpeed=5
    maxGustDuration=2
    maxGustStrength=7
    maxTimeBetweenGusts=7
    maxUpInterpSpeed=5
    minDownInterpSpeed=2
    minGustDuration=0.5
    minGustStrength=3
    minTimeBetweenGusts=0
    minUpInterpSpeed=2
    windTurbulenceScalarA=0.1
    windTurbulenceScalarB=1
    windAmplitude=<0 0 0>
    windFrequency=<0 0 0>
    windOffset=<0 0 0>
   }
   tactical_Sprint {
    maxDownInterpSpeed=5
    maxGustDuration=2
    maxGustStrength=7
    maxTimeBetweenGusts=7
    maxUpInterpSpeed=5
    minDownInterpSpeed=2
    minGustDuration=0.5
    minGustStrength=3
    minTimeBetweenGusts=0
    minUpInterpSpeed=2
    windTurbulenceScalarA=0.1
    windTurbulenceScalarB=1
    windAmplitude=<0 0 0>
    windFrequency=<0 0 0>
    windOffset=<0 0 0>
   }
   ziplining {
    maxDownInterpSpeed=5
    maxGustDuration=2
    maxGustStrength=7
    maxTimeBetweenGusts=7
    maxUpInterpSpeed=5
    minDownInterpSpeed=2
    minGustDuration=0.5
    minGustStrength=3
    minTimeBetweenGusts=0
    minUpInterpSpeed=2
    windTurbulenceScalarA=0.1
    windTurbulenceScalarB=1
    windAmplitude=<0 0 0>
    windFrequency=<0 0 0>
    windOffset=<0 0 0>
   }
  }
 }
 WPO Pho Wind Intensity {type=number}
]
