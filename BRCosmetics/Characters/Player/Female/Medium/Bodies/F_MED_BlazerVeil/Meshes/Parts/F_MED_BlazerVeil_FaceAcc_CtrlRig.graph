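# Control Rig for the BlazerVeil female medium cosmetic's face accessory, applying corrective transforms to dynamic hair and strand bones.
# A forward chain scales a front-left strand bone, then rotates hair bones weighted by the Skydive variable.
# Four spherical pose readers track head orientation against neck_01 to weight further hair corrections.
# AlphaInterp nodes condition three of these weights, with ADS scaling two and crouch capping the last.
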ControlRigBlueprint /BRCosmetics/Characters/Player/Female/Medium/Bodies/F_MED_BlazerVeil/Meshes/Parts/F_MED_BlazerVeil_FaceAcc_CtrlRig
graphs [
 RigVMFunctionLibrary {links [], nodes []}
 RigVMModel {nodes [
   RigUnit_BeginExecution = &RigUnit_BeginExecution()
   &RigUnit_ModifyTransforms(
    ItemToModify=((Item=(Type=Bone,Name="dyn_strand_fr_l_04"),Transform=(Rotation=(X=0,Y=0,Z=0,W=1),Translation=(X=0,Y=0,Z=0),Scale3D=(X=0.85,Y=1,Z=1))),(Item=(Type=Bone,Name="None"),Transform=(Rotation=(X=0,Y=0,Z=0,W=1),Translation=(X=0,Y=0,Z=0),Scale3D=(X=0.5,Y=1,Z=1))))
    Weight=1
    ExecutePin=@RigUnit_BeginExecution.ExecutePin
   )
   > &RigUnit_ModifyTransforms(
    ItemToModify="((Item=(Type=Bone,Name=\"dyn_hair_side_l_01\"),Transform=(Rotation=(X=0,Y=0,Z=0,W=1),Translation=(X=0,Y=0,Z=0),Scale3D=(X=1,Y=1,Z=1))),(Item=(Type=Bone,Name=\"dyn_hair_bk_l_01\"),Transform=(Rotation=(X=0,Y=0,Z=0,W=1),Translation=(X=0,Y=0,Z=0),Scale3D=(X=1,Y=1, ...(truncated)"
    ItemToModify.0.Transform.Rotation=&RigVMFunction_MathQuaternionFromEuler(Euler=(X=0,Y=0,Z=10), RotationOrder=XYZ).Result
    ItemToModify.1.Transform.Rotation=&RigVMFunction_MathQuaternionFromEuler(Euler=(X=0,Y=0,Z=10), RotationOrder=XYZ).Result
    ItemToModify.2.Transform.Rotation=&RigVMFunction_MathQuaternionFromEuler(Euler=(X=0,Y=0,Z=10), RotationOrder=XYZ).Result
    ItemToModify.4.Transform.Rotation=&RigVMFunction_MathQuaternionFromEuler(Euler=(X=0,Y=0,Z=10), RotationOrder=XYZ).Result
    ItemToModify.3.Transform.Rotation=&RigVMFunction_MathQuaternionFromEuler(Euler=(X=0,Y=0,Z=10), RotationOrder=XYZ).Result
    Weight=RigVMVariableNode() {variable=Skydive}.Value
   )
   > n1 = &RigUnit_SphericalPoseReader(
    ActiveRegionScaleFactors=(PositiveWidth=0.48,NegativeWidth=0.2,PositiveHeight=1,NegativeHeight=1)
    ActiveRegionSize=0.2
    Debug=(bDrawDebug=false,DebugScale=25,DebugSegments=20,DebugThickness=0.25)
    DriverAxis=(X=1,Y=0,Z=0)
    DriverItem=(Type=Bone,Name="head")
    FalloffRegionScaleFactors=(PositiveWidth=0.28,NegativeWidth=0.1,PositiveHeight=0.4,NegativeHeight=0.2)
    FalloffSize=0.5
    OptionalParentItem=(Type=Bone,Name="neck_01")
    RotationOffset=(X=-90,Y=0,Z=55)
   )
   AlphaInterp = &RigVMFunction_AlphaInterp(
    InterpSpeedDecreasing=10
    InterpSpeedIncreasing=10
    bClampResult=False
    bInterpResult=False
    Value=@n1.OutputParam
    Scale=RigVMVariableNode() {variable=ADS}.Value
   )
   > &RigUnit_ModifyTransforms(
    ItemToModify="((Item=(Type=Bone,Name=\"dyn_hair_bk_r_a_01\"),Transform=(Rotation=(X=0,Y=0,Z=0,W=1),Translation=(X=0,Y=0,Z=0),Scale3D=(X=1,Y=1,Z=1))),(Item=(Type=Bone,Name=\"dyn_hair_br_r_b_01\"),Transform=(Rotation=(X=0,Y=0,Z=0,W=1),Translation=(X=0,Y=0,Z=0),Scale3D=(X=1,Y= ...(truncated)"
    Weight=@AlphaInterp.Result
    ItemToModify.0.Transform.Rotation=&RigVMFunction_MathQuaternionFromEuler(Euler=(X=0,Y=0,Z=0), RotationOrder=XYZ).Result
    ItemToModify.1.Transform.Rotation=&RigVMFunction_MathQuaternionFromEuler(Euler=(X=0,Y=20,Z=0), RotationOrder=XYZ).Result
   )
   > n2 = &RigUnit_SphericalPoseReader(
    ActiveRegionScaleFactors=(PositiveWidth=0.45,NegativeWidth=0.2,PositiveHeight=1,NegativeHeight=1)
    ActiveRegionSize=0.2
    Debug=(bDrawDebug=false,DebugScale=25,DebugSegments=20,DebugThickness=0.25)
    DriverAxis=(X=1,Y=0,Z=0)
    DriverItem=(Type=Bone,Name="head")
    FalloffRegionScaleFactors=(PositiveWidth=0.22,NegativeWidth=0.1,PositiveHeight=0.4,NegativeHeight=0.2)
    FalloffSize=0.5
    OptionalParentItem=(Type=Bone,Name="neck_01")
    RotationOffset=(X=-90,Y=0,Z=55)
   )
   > &RigUnit_ModifyTransforms(
    ItemToModify="((Item=(Type=Bone,Name=\"dyn_strand_fr_r_01\"),Transform=(Rotation=(X=0,Y=0,Z=0,W=1),Translation=(X=0,Y=0,Z=0),Scale3D=(X=1,Y=1,Z=1))),(Item=(Type=Bone,Name=\"dyn_strand_fr_l_04\"),Transform=(Rotation=(X=0,Y=0,Z=0,W=1),Translation=(X=0,Y=0,Z=0),Scale3D=(X=0.5, ...(truncated)"
    ItemToModify.0.Transform.Rotation=&RigVMFunction_MathQuaternionFromEuler(Euler=(X=0,Y=-9,Z=0), RotationOrder=XYZ).Result
    Weight=@n2.OutputParam
   )
   > n3 = &RigUnit_SphericalPoseReader(
    ActiveRegionScaleFactors=(PositiveWidth=0.1,NegativeWidth=0.2,PositiveHeight=1,NegativeHeight=1)
    ActiveRegionSize=0.2
    Debug=(bDrawDebug=false,bDraw2D=false,bDrawLocalAxes=false,DebugScale=25,DebugSegments=20,DebugThickness=0.25)
    DriverAxis=(X=1,Y=0,Z=0)
    DriverItem=(Type=Bone,Name="head")
    FalloffRegionScaleFactors=(PositiveWidth=0.1,NegativeWidth=0.25,PositiveHeight=0.2,NegativeHeight=0.2)
    FalloffSize=0.45
    OptionalParentItem=(Type=Bone,Name="neck_01")
    RotationOffset=(X=-90,Y=0,Z=125)
   )
   n4 = &RigVMFunction_AlphaInterp(InterpSpeedDecreasing=7, InterpSpeedIncreasing=5, Scale=1, bClampResult=False, Value=@n3.OutputParam)
   > &RigUnit_ModifyTransforms(
    ItemToModify="((Item=(Type=Bone,Name=\"dyn_strand_fr_r_01\"),Transform=(Rotation=(X=0,Y=0,Z=0,W=1),Translation=(X=0,Y=0,Z=0),Scale3D=(X=1,Y=1,Z=1))),(Item=(Type=Bone,Name=\"dyn_strand_fr_r_03\"),Transform=(Rotation=(X=0,Y=0,Z=0,W=1),Translation=(X=0,Y=0,Z=0),Scale3D=(X=1,Y= ...(truncated)"
    Weight=@n4.Result
    ItemToModify.0.Transform.Rotation=&RigVMFunction_MathQuaternionFromEuler(Euler=(X=0,Y=-5,Z=0), RotationOrder=XYZ).Result
    ItemToModify.1.Transform.Rotation=&RigVMFunction_MathQuaternionFromEuler(Euler=(X=0,Y=-15,Z=0), RotationOrder=XYZ).Result
   )
   > n5 = &RigUnit_SphericalPoseReader(
    ActiveRegionScaleFactors=(PositiveWidth=0.3,NegativeWidth=0.2,PositiveHeight=1,NegativeHeight=1)
    ActiveRegionSize=0.2
    Debug=(bDrawDebug=true,DebugScale=25,DebugSegments=20,DebugThickness=0.25)
    DriverAxis=(X=1,Y=0,Z=0)
    DriverItem=(Type=Bone,Name="head")
    FalloffRegionScaleFactors=(PositiveWidth=0.1,NegativeWidth=0.1,PositiveHeight=0.2,NegativeHeight=0.2)
    FalloffSize=0.5
    OptionalParentItem=(Type=Bone,Name="neck_01")
    RotationOffset=(X=-90,Y=0,Z=55)
   )
   n6 = &RigVMFunction_AlphaInterp(
    InterpSpeedDecreasing=7
    InterpSpeedIncreasing=5
    bClampResult=true
    Value=@n5.OutputParam
    Scale=RigVMVariableNode() {variable=ADS}.Value
    ClampMax=RigVMVariableNode() {variable=crouch}.Value
   )
   > &RigUnit_ModifyTransforms(
    ItemToModify="((Item=(Type=Bone,Name=\"dyn_hair_r_01\"),Transform=(Rotation=(X=0,Y=0,Z=0,W=1),Translation=(X=0,Y=0,Z=0),Scale3D=(X=1,Y=1,Z=1))),(Item=(Type=Bone,Name=\"dyn_hair_bk_r_a_01\"),Transform=(Rotation=(X=0,Y=0,Z=0,W=1),Translation=(X=0,Y=0,Z=0),Scale3D=(X=1,Y=1,Z=1 ...(truncated)"
    Weight=@n6.Result
    ItemToModify.0.Transform.Rotation=&RigVMFunction_MathQuaternionFromEuler(Euler=(X=0,Y=20,Z=0), RotationOrder=XYZ).Result
    ItemToModify.1.Transform.Rotation=&RigVMFunction_MathQuaternionFromEuler(Euler=(X=0,Y=0,Z=25), RotationOrder=XYZ).Result
    ItemToModify.2.Transform.Rotation=&RigVMFunction_MathQuaternionFromEuler(Euler=(X=0,Y=0,Z=25), RotationOrder=XYZ).Result
    ItemToModify.3.Transform.Rotation=&RigVMFunction_MathQuaternionFromEuler(Euler=(X=0,Y=0,Z=25), RotationOrder=XYZ).Result
   )
  ]}
]
hierarchy [
 root {loc=<0 0 0>, rot=<0 0 0 1>}
 pelvis {parent=root, loc=<0 -0.5301 98.0571>, rot=<0.0127 0.707 -0.0127 -0.707>}
 spine_01 {parent=pelvis, loc=<6.1208 -0.9328 0>, rot=<0 0 0.034 -0.9994>}
 spine_02 {parent=spine_01, loc=<6.5996 -0.2095 0>, rot=<0 0 0.0238 -0.9997>}
 spine_03 {parent=spine_02, loc=<6.4914 -0.1237 0>, rot=<0 0 0.0149 -0.9999>}
 spine_04 {parent=spine_03, loc=<6.9598 0.0794 0>, rot=<0 0 -0.0132 -0.9999>}
 spine_05 {parent=spine_04, loc=<6.9987 0 0>, rot=<0 0 -0.0015 -1>}
 neck_01 {parent=spine_05, loc=<11.3138 0.1529 0>, rot=<0 0 -0.2003 -0.9797>}
 neck_02 {parent=neck_01, loc=<4.5627 0 0>, rot=<0 0 0.0504 -0.9987>}
 head {parent=neck_02, loc=<5.3914 0 0>, rot=<0 0 0.115 -0.9934>}
 dyn_earring_l {parent=head, loc=<-0.1731 -0.2154 -6.4361>, rot=<0.0162 -0.7068 0.7071 0.0154>}
 dyn_earring_r {parent=head, loc=<-0.3086 -0.1077 6.3507>, rot=<0.0191 -0.707 -0.7069 0.0024>}
 dyn_strand_fr_l_01 {parent=head, loc=<6.1366 -0.6313 -6.4368>, rot=<0.2316 -0.5722 0.7341 -0.2828>}
 dyn_strand_fr_l_02 {parent=dyn_strand_fr_l_01, loc=<2.5052 0 0>, rot=<-0.0207 -0.2091 0.0181 -0.9775>}
 dyn_strand_fr_l_03 {parent=dyn_strand_fr_l_02, loc=<2.2701 0 0>, rot=<-0.004 -0.161 0.0167 -0.9868>}
 dyn_strand_fr_l_04 {parent=dyn_strand_fr_l_03, loc=<2.4626 0 0>, rot=<0.0008 -0.0759 0.0082 -0.9971>}
 dyn_strand_fr_l_05 {parent=dyn_strand_fr_l_04, loc=<2.2357 0 0>, rot=<-0.0383 -0.1153 0.1568 -0.9801>}
 dyn_strand_l_01 {parent=head, loc=<6.8606 -3.2554 -7.4288>, rot=<0.079 -0.5771 0.8071 -0.0961>}
 dyn_strand_l_02 {parent=dyn_strand_l_01, loc=<4.7639 0.0382 -0.1766>, rot=<-0.0033 -0.0287 -0.0061 -0.9996>}
 dyn_strand_l_03 {parent=dyn_strand_l_02, loc=<2.2909 0 0>, rot=<-0.0039 -0.0523 -0.0107 -0.9986>}
 dyn_hair_side_l_01 {parent=head, loc=<8.7532 -5.4591 -5.3084>, rot=<0.0103 -0.2508 0.9666 -0.0523>}
 dyn_hair_side_l_02 {parent=dyn_hair_side_l_01, loc=<2.735 0 0>, rot=<0 0 0 1>}
 dyn_hair_side_l_03 {parent=dyn_hair_side_l_02, loc=<4.0697 0 0>, rot=<0 0 0 1>}
 dyn_hair_bk_l_01 {parent=head, loc=<8.7921 -6.6402 -1.3071>, rot=<0.0169 0.0059 0.9995 -0.0245>}
 dyn_hair_bk_l_02 {parent=dyn_hair_bk_l_01, loc=<3.2821 0 0>, rot=<0 0 0 1>}
 dyn_hair_bk_l_03 {parent=dyn_hair_bk_l_02, loc=<3.5284 0 0>, rot=<0 0 0 1>}
 dyn_hair_bk_r_01 {parent=head, loc=<9.0296 -6.759 1.8817>, rot=<0.0349 0.2256 0.9735 -0.0121>}
 dyn_hair_bk_r_02 {parent=dyn_hair_bk_r_01, loc=<3.6471 0 0>, rot=<0 0 0 1>}
 dyn_hair_bk_r_03 {parent=dyn_hair_bk_r_02, loc=<3.1581 0 0>, rot=<0 0 0 1>}
 dyn_hair_bk_r_a_01 {parent=head, loc=<9.1589 -5.62 4.8721>, rot=<0.0156 0.4203 0.9072 0.0022>}
 dyn_hair_bk_r_a_02 {parent=dyn_hair_bk_r_a_01, loc=<2.7048 0 0>, rot=<0 0 0 1>}
 dyn_hair_bk_r_a_03 {parent=dyn_hair_bk_r_a_02, loc=<4.3611 0 0>, rot=<0 0 0 1>}
 dyn_hair_br_r_b_01 {parent=head, loc=<8.873 -2.3256 7.2712>, rot=<-0.0174 0.5056 0.8623 -0.0202>}
 dyn_hair_br_r_b_02 {parent=dyn_hair_br_r_b_01, loc=<3.6337 0 0>, rot=<0 0 0 1>}
 dyn_hair_br_r_b_03 {parent=dyn_hair_br_r_b_02, loc=<3.9856 0 0>, rot=<0 0 0 1>}
 dyn_hair_r_01 {parent=head, loc=<8.2868 2.689 7.9459>, rot=<0.0032 -0.7125 -0.6943 0.1015>}
 dyn_hair_r_02 {parent=dyn_hair_r_01, loc=<4.0196 0 0>, rot=<0 0 0 1>}
 dyn_hair_r_03 {parent=dyn_hair_r_02, loc=<4.3955 0 0>, rot=<0 0 0 1>}
 dyn_hair_side_r_01 {parent=head, loc=<7.287 6.8318 6.1379>, rot=<0.0203 -0.8083 -0.5649 0.1646>}
 dyn_hair_side_r_02 {parent=dyn_hair_side_r_01, loc=<4.5091 0 0>, rot=<0 0 0 1>}
 dyn_hair_side_r_03 {parent=dyn_hair_side_r_02, loc=<4.2817 0 0>, rot=<0 0 0 1>}
 dyn_strand_r_01 {parent=head, loc=<5.794 9.8296 5.1061>, rot=<0.0238 -0.9031 -0.4278 0.0292>}
 dyn_strand_r_02 {parent=dyn_strand_r_01, loc=<3.9602 0 0>, rot=<0 0 0 1>}
 dyn_strand_r_03 {parent=dyn_strand_r_02, loc=<5.221 0 0>, rot=<0 0 0 1>}
 dyn_strand_fr_r_01 {parent=head, loc=<11.1746 9.7895 -1.113>, rot=<-0.1246 -0.9529 -0.1691 0.2188>}
 dyn_strand_fr_r_02 {parent=dyn_strand_fr_r_01, loc=<2.7491 0 0>, rot=<0 0 0 1>}
 dyn_strand_fr_r_03 {parent=dyn_strand_fr_r_02, loc=<2.9787 0 0>, rot=<0 0 0 1>}
 dyn_strand_fr_r_04 {parent=dyn_strand_fr_r_03, loc=<2.8197 0 0>, rot=<0 0 0 1>}
 dyn_strand_fr_r_05 {parent=dyn_strand_fr_r_04, loc=<2.1336 0 0>, rot=<0 0 0 1>}
]
variables [
 ADS {type=double, default=0}
 Skydive {type=double, default=0}
 crouch {type=double, default=0}
]
